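# Niagara system for the Road Sparks drift trail vehicle cosmetic, spraying short-lived sprites backward from the wheels.
# Cone and cloud emitters place particles at skeletal-mesh sockets by partition, with per-tire FX disabling.
# Spawn groups index arrays for size, lifetime, shape and velocity, and set a visibility tag, letting one emitter carry several layers.
# Painted-variant overrides supply colors per paint and renderer, while game mode ID switches speed ranges.
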NiagaraSystem /VehicleCosmetics/DriftTrail/DriftTrail_RoadSparks/FX/NS_DriftTrail_RoadSparks
dependencies {
 usedDataInterfaces [
  &NiagaraDataInterfaceCurve
  &NiagaraDataInterfaceArrayFloat2
  &NiagaraDataInterfaceArrayFloat3
  &NiagaraDataInterfaceColorCurve
  &NiagaraDataInterfaceSpriteRendererInfo
  &NiagaraDataInterfaceVectorCurve
  &NiagaraDataInterfaceSkeletalMesh
 ]
 usedDynamicInputs [&VC_Mask_Vectors, &ScaleLinearColorByCurve]
 usedModules [
  &self:SystemSpawnScript.SystemScriptSource.SystemScriptGraph.NiagaraNodeAssignment_0.SetVariables_#1
  &self:SystemSpawnScript.SystemScriptSource.SystemScriptGraph.NiagaraNodeAssignment_1.SetVariables_#2
  &self:SystemSpawnScript.SystemScriptSource.SystemScriptGraph.NiagaraNodeAssignment_2.SetVariables_#3
  &FN_UnpackVehicleFlags
  &SystemState
  /Game/Effects/Niagara/Modules/Vehicle/FN_UpdateVehicleAttribs
 ]
}
emitters {
 NE_VC_DriftTrail_Cone {
  emitterUpdateScript [
   &EmitterState()
   &SpawnRate(SpawnRate=220, Use Spawn Probability=false)
   &SpawnRate(SpawnGroup=1, SpawnRate=500, Use Spawn Probability=true) {module=SpawnRate001}
   &SpawnRate(SpawnGroup=2, SpawnRate=76, Use Spawn Probability=false) {module=SpawnRate003}
   &SpawnRate(SpawnGroup=3, SpawnRate=100, Use Spawn Probability=false) {module=SpawnRate004}
  ]
  particleSpawnScript [
   &VC_PaintedVariantsOverride(DefaultColor=<rgba 0 0 0 3>, R1 DefaultColor=<rgba 0.4233 0.3549 0.2822 1>, Use R1 Painted Overrides=true) {enabled=false}
   /SetVariables(Particles.VisibilityTag=$Emitter.SpawnGroup) {isSetParameters=true, module=SetVariables_#4}
   &InitializeParticle(
    Color Mode=::NewEnumerator0
    Lifetime Max=&SelectFloatFromArray(
     Direct Array Index=&SpawnGroup()
     Float Selection Array=DataInterface() {propertyValues {FloatData [0.14,0.16,0.1,0.15]}}
    )
    Lifetime Min:NiagaraFloat=&SelectFloatFromArray(
     Direct Array Index=&SpawnGroup()
     Float Selection Array=DataInterface() {propertyValues {FloatData [0.12,0.12,0.03,0.1]}}
    )
    Lifetime Mode=::NewEnumerator1
    Position Offset=<-50 0 25>
    Sprite Rotation Mode=::NewEnumerator2
    Sprite Size Max=&SelectVector2DFromArray(
     Direct Array Index=&SpawnGroup()
     Vector 2D Selection Array=DataInterface() {propertyValues {FloatData [<4 22.5>, <5.5 25>, <0.7 37>, <4 6>]}}
    )
    Sprite Size Min=&SelectVector2DFromArray(
     Direct Array Index=&SpawnGroup()
     Vector 2D Selection Array=DataInterface() {propertyValues {FloatData [<3.5 8>, <3.75 8>, <0.5 17>, <2 3>]}}
    )
    Sprite Size Mode=::NewEnumerator2
    UsePositionOffset=false
    Write Material Random=true
   )
   &PartitionParticles()
   &SkeletalMeshLocation(
    Bone / Socket Index=&Add_Integer(A=$Particles.PartitionParticles.Partition, B=4)
    Bone / Socket Sampling Mode=::NewEnumerator5
    Interpolate Mesh Sampling=false
    MaskSpawnGroup=false
    Mesh Sampling Type=::NewEnumerator6
    Skeletal Mesh=$System.SkeletalMesh
   )
   &AddVelocity(
    Cone Angle=&MaskFloatBySpawnGroup(
     Masked Float=90
     Passthrough Float=&MaskFloatBySpawnGroup(
      Masked Float=96
      Passthrough Float=&MaskFloatBySpawnGroup(Masked Float=35, Passthrough Float=40, Spawn Group Mask=2)
      Spawn Group Mask=1
     )
    )
    Cone Axis=&RandomRangeVector(
     Minimum=&SelectVectorFromArray(
      Direct Array Index=&SpawnGroup()
      Vector Selection Array=DataInterface() {propertyValues {FloatData [<-1 -0.225 -0.1>, <-1 -0.275 -0.1>, <-1 0 -0.05>, <-1 0 0.2>]}}
     )
     Maximum=&SelectVectorFromArray(
      Direct Array Index=&SpawnGroup()
      Vector Selection Array=DataInterface() {propertyValues {FloatData [<-0.45 0.225 0.65>, <-1 0.275 0.6>, <-0.9 0 0.3>, <-0.075 0 0.5>]}}
     )
    )
    Distribution Along Cone Axis=1
    Inner Cone Angle=25
    Module.UseConeFalloff=false
    Module.UseVelDistribution=true
    Override Seed=false
    Velocity Mode=::NewEnumerator2
    Velocity Speed=&MergeHelper_FloatSwitch(
     0=&RandomRangeFloat(
      Minimum=&SelectFloatFromArray(
       Direct Array Index=&SpawnGroup()
       Float Selection Array=DataInterface() {propertyValues {FloatData [400,425,150,250]}}
      )
      Maximum=&SelectFloatFromArray(
       Direct Array Index=&SpawnGroup()
       Float Selection Array=DataInterface() {propertyValues {FloatData [700,750,1150,1150]}}
      )
     )
     1=&RandomRangeFloat(
      Minimum=&SelectFloatFromArray(
       Direct Array Index=&SpawnGroup()
       Float Selection Array=DataInterface() {propertyValues {FloatData [400,425,150,250]}}
      )
      Maximum=&SelectFloatFromArray(
       Direct Array Index=&SpawnGroup()
       Float Selection Array=DataInterface() {propertyValues {FloatData [700,750,950,1150]}}
      )
     )
     2=&RandomRangeFloat(
      Minimum=&SelectFloatFromArray(
       Direct Array Index=&SpawnGroup()
       Float Selection Array=DataInterface() {propertyValues {FloatData [400,425,150,250]}}
      )
      Maximum=&SelectFloatFromArray(
       Direct Array Index=&SpawnGroup()
       Float Selection Array=DataInterface() {propertyValues {FloatData [700,750,1150,1150]}}
      )
     )
     3=&RandomRangeFloat(
      Minimum=&SelectFloatFromArray(
       Direct Array Index=&SpawnGroup()
       Float Selection Array=DataInterface() {propertyValues {FloatData [400,425,150,250]}}
      )
      Maximum=&SelectFloatFromArray(
       Direct Array Index=&SpawnGroup()
       Float Selection Array=DataInterface() {propertyValues {FloatData [700,750,950,1150]}}
      )
     )
     Index=$System.GameModeID
     Num=::NewEnumerator2
    )
   )
   &ShapeLocation(
    Cylinder Height=&SelectFloatFromArray(Direct Array Index=&SpawnGroup(), Float Selection Array=DataInterface() {propertyValues {FloatData [5,5,5,14]}})
    Cylinder Radius=&SelectFloatFromArray(
     Direct Array Index=&SpawnGroup()
     Float Selection Array=DataInterface() {propertyValues {FloatData [15,12.5,17,14]}}
    )
    Hemicircle Y=true
    MaskSpawnGroup=false
    Offset=&SelectVectorFromArray(
     Direct Array Index=&SpawnGroup()
     Vector Selection Array=DataInterface() {propertyValues {FloatData [<-43 0 -31>, <-43 0 -35>, <-50 0 -31.5>, <-60 0 -15>]}}
    )
    Offset Coordinate Space=::Local
    Offset Mode=::NewEnumerator0
    Override Seed=false
    Rotation Angle=0.25
    Rotation Mode=::NewEnumerator1
    Shape Primitive=::NewEnumerator1
    Surface Only=true
    Use Endcaps In Surface Only Mode=false
    UseOwnerScale=true
   ) {module=ShapeLocation001}
  ]
  particleUpdateScript [
   &VC_PaintedVariantsOverride(
    Black=<rgba 0 0 0 3>
    BurntSienna=&MultiplyLinearColor(
     Linear Color=<rgba 0.126 0.0595 0.0238 1>
     Scale Factor=&MakeVector4FromVectorAndFloat("Vector (RGB)"=&VectorFromFloat(Value=1.2), "Float (Alpha)"=2.25)
    )
    Cobalt=&MultiplyLinearColor(
     Linear Color=<rgba 0.05 0.1 1 1>
     Scale Factor=&MakeVector4FromVectorAndFloat("Vector (RGB)"=&VectorFromFloat(Value=0.2), "Float (Alpha)"=2.2)
    )
    Crimson=&MultiplyLinearColor(
     Linear Color=<rgba 0.5 0 0 1>
     Scale Factor=&MakeVector4FromVectorAndFloat("Vector (RGB)"=&VectorFromFloat(Value=0.4), "Float (Alpha)"=2.2)
    )
    DefaultColor=<rgba 0 0 0 3>
    ForestGreen=&MultiplyLinearColor(
     Linear Color=<rgba 0 0.35 0 1>
     Scale Factor=&MakeVector4FromVectorAndFloat("Vector (RGB)"=&VectorFromFloat(Value=0.4), "Float (Alpha)"=2.2)
    )
    Grey=&MultiplyLinearColor(
     Linear Color=<rgba 0.125 0.125 0.125 1>
     Scale Factor=&MakeVector4FromVectorAndFloat("Vector (RGB)"=&VectorFromFloat(Value=1.25), "Float (Alpha)"=2.2)
    )
    Lime=&MultiplyLinearColor(
     Linear Color=<rgba 0.3 0.7 0 1>
     Scale Factor=&MakeVector4FromVectorAndFloat("Vector (RGB)"=&VectorFromFloat(Value=0.35), "Float (Alpha)"=2.2)
    )
    Orange=&MultiplyLinearColor(
     Linear Color=<rgba 0.7 0.35 0.01 1>
     Scale Factor=&MakeVector4FromVectorAndFloat("Vector (RGB)"=&VectorFromFloat(Value=0.4), "Float (Alpha)"=2.2)
    )
    Pink=&MultiplyLinearColor(
     Linear Color=<rgba 0.55 0.145 0.4 1>
     Scale Factor=&MakeVector4FromVectorAndFloat("Vector (RGB)"=&VectorFromFloat(Value=0.4), "Float (Alpha)"=2.2)
    )
    Purple=&MultiplyLinearColor(
     Linear Color=<rgba 0.25 0 0.3 1>
     Scale Factor=&MakeVector4FromVectorAndFloat("Vector (RGB)"=&VectorFromFloat(Value=0.5), "Float (Alpha)"=2.2)
    )
    R1 Black:LinearColor=<rgba 0.15 0.15 0.15 1.5>
    R1 BurntSienna:LinearColor=&MultiplyLinearColor(
     Linear Color=<rgba 0.21 0.099 0.0454 1>
     Scale Factor=&MakeVector4FromVectorAndFloat("Vector (RGB)"=&VectorFromFloat(Value=1.8), "Float (Alpha)"=0.98)
    )
    R1 Cobalt:LinearColor=&MultiplyLinearColor(
     Linear Color=<rgba 0.125 0.2 0.875 1>
     Scale Factor=&MakeVector4FromVectorAndFloat("Vector (RGB)"=&VectorFromFloat(Value=1.31), "Float (Alpha)"=0.98)
    )
    R1 Crimson:LinearColor=&MultiplyLinearColor(
     Linear Color=<rgba 0.5 0.15 0.15 1>
     Scale Factor=&MakeVector4FromVectorAndFloat("Vector (RGB)"=&VectorFromFloat(Value=1), "Float (Alpha)"=1.08)
    )
    R1 DefaultColor=<rgba 0.4233 0.402 0.3813 0.95>
    R1 ForestGreen:LinearColor=&MultiplyLinearColor(
     Linear Color=<rgba 0.05 0.35 0.05 1>
     Scale Factor=&MakeVector4FromVectorAndFloat("Vector (RGB)"=&VectorFromFloat(Value=1.5), "Float (Alpha)"=1)
    )
    R1 Grey:LinearColor=&MultiplyLinearColor(
     Linear Color=<rgba 0.125 0.125 0.14 1>
     Scale Factor=&MakeVector4FromVectorAndFloat("Vector (RGB)"=&VectorFromFloat(Value=2.85), "Float (Alpha)"=1)
    )
    R1 Lime:LinearColor=&MultiplyLinearColor(
     Linear Color=<rgba 0.6 0.7 0.05 1>
     Scale Factor=&MakeVector4FromVectorAndFloat("Vector (RGB)"=&VectorFromFloat(Value=1.2), "Float (Alpha)"=0.8)
    )
    R1 Orange:LinearColor=&MultiplyLinearColor(
     Linear Color=<rgba 0.7 0.35 0.01 1>
     Scale Factor=&MakeVector4FromVectorAndFloat("Vector (RGB)"=&VectorFromFloat(Value=1.25), "Float (Alpha)"=1)
    )
    R1 Pink:LinearColor=&MultiplyLinearColor(
     Linear Color=<rgba 0.5 0.275 0.47 1>
     Scale Factor=&MakeVector4FromVectorAndFloat("Vector (RGB)"=&VectorFromFloat(Value=1.4), "Float (Alpha)"=1.08)
    )
    R1 Purple:LinearColor=&MultiplyLinearColor(
     Linear Color=<rgba 0.122 0.096 0.24 1>
     Scale Factor=&MakeVector4FromVectorAndFloat("Vector (RGB)"=&VectorFromFloat(Value=2.55), "Float (Alpha)"=1.31)
    )
    R1 Saffron:LinearColor=&MultiplyLinearColor(
     Linear Color=<rgba 0.5 0.45 0.1 1>
     Scale Factor=&MakeVector4FromVectorAndFloat("Vector (RGB)"=&VectorFromFloat(Value=1.65), "Float (Alpha)"=1)
    )
    R1 SkyBlue:LinearColor=&MultiplyLinearColor(
     Linear Color=<rgba 0.05 0.5 0.8 1>
     Scale Factor=&MakeVector4FromVectorAndFloat("Vector (RGB)"=&VectorFromFloat(Value=1.2), "Float (Alpha)"=1.1)
    )
    R1 TitaniumWhite:LinearColor=&MultiplyLinearColor(
     Linear Color=<rgba 1 1 1.05 1>
     Scale Factor=&MakeVector4FromVectorAndFloat("Vector (RGB)"=&VectorFromFloat(Value=0.95), "Float (Alpha)"=0.85)
    )
    R3 Black:LinearColor=&MultiplyLinearColor(
     Linear Color=<rgba 0.15 0.15 0.15 0.5>
     Scale Factor=&MakeVector4FromVectorAndFloat(
      "Vector (RGB)"=<6.1 6.1 6.1>
      "Float (Alpha)"=&FloatFromCurve(FloatCurve=DataInterface() {propertyValues {Curve {keys [<at 0.55 1>,<at 1 0>]}}})
     )
    )
    R3 BurntSienna:LinearColor=&MultiplyLinearColor(
     Linear Color=<rgba 0.21 0.0992 0.0396 1>
     Scale Factor=&MakeVector4FromVectorAndFloat(
      "Vector (RGB)"=<5.1 5.35 5.1>
      "Float (Alpha)"=&FloatFromCurve(FloatCurve=DataInterface() {propertyValues {Curve {keys [<at 0.55 1>,<at 1 0>]}}})
     )
    )
    R3 Cobalt:LinearColor=&MultiplyLinearColor(
     Linear Color=<rgba 0.1 0.2 0.95 1>
     Scale Factor=&MakeVector4FromVectorAndFloat(
      "Vector (RGB)"=<2.35 2.35 2.35>
      "Float (Alpha)"=&FloatFromCurve(FloatCurve=DataInterface() {propertyValues {Curve {keys [<at 0.55 1>,<at 1 0>]}}})
     )
    )
    R3 Crimson:LinearColor=&MultiplyLinearColor(
     Linear Color=<rgba 0.5 0.175 0.175 1>
     Scale Factor=&MakeVector4FromVectorAndFloat(
      "Vector (RGB)"=<2.5 2.5 2.5>
      "Float (Alpha)"=&FloatFromCurve(FloatCurve=DataInterface() {propertyValues {Curve {keys [<at 0.55 1>,<at 1 0>]}}})
     )
    )
    R3 DefaultColor:LinearColor=&MultiplyLinearColor(
     Linear Color=<rgba 10 5.6264 1.4583 0.4>
     Scale Factor=&MakeVector4FromVectorAndFloat(
      "Vector (RGB)"=<1 1 1>
      "Float (Alpha)"=&FloatFromCurve(FloatCurve=DataInterface() {propertyValues {Curve {keys [<at 0.55 1>,<at 1 0>]}}})
     )
    )
    R3 ForestGreen:LinearColor=&MultiplyLinearColor(
     Linear Color=<rgba 0.15 0.35 0.15 1>
     Scale Factor=&MakeVector4FromVectorAndFloat(
      "Vector (RGB)"=<2.55 2.55 2.55>
      "Float (Alpha)"=&FloatFromCurve(FloatCurve=DataInterface() {propertyValues {Curve {keys [<at 0.55 1>,<at 1 0>]}}})
     )
    )
    R3 Grey:LinearColor=&MultiplyLinearColor(
     Linear Color=<rgba 0.125 0.125 0.14 1>
     Scale Factor=&MakeVector4FromVectorAndFloat(
      "Vector (RGB)"=<5 5 5>
      "Float (Alpha)"=&FloatFromCurve(FloatCurve=DataInterface() {propertyValues {Curve {keys [<at 0.55 1>,<at 1 0>]}}})
     )
    )
    R3 Lime:LinearColor=&MultiplyLinearColor(
     Linear Color=<rgba 0.6 0.7 0.05 1>
     Scale Factor=&MakeVector4FromVectorAndFloat(
      "Vector (RGB)"=<2.55 2.55 2.55>
      "Float (Alpha)"=&FloatFromCurve(FloatCurve=DataInterface() {propertyValues {Curve {keys [<at 0.55 1>,<at 1 0>]}}})
     )
    )
    R3 Orange:LinearColor=&MultiplyLinearColor(
     Linear Color=<rgba 0.7 0.4 0.25 1>
     Scale Factor=&MakeVector4FromVectorAndFloat(
      "Vector (RGB)"=<2.5 2.5 2.5>
      "Float (Alpha)"=&FloatFromCurve(FloatCurve=DataInterface() {propertyValues {Curve {keys [<at 0.55 1>,<at 1 0>]}}})
     )
    )
    R3 Pink:LinearColor=&MultiplyLinearColor(
     Linear Color=<rgba 0.6 0.33 0.564 1>
     Scale Factor=&MakeVector4FromVectorAndFloat(
      "Vector (RGB)"=<2.5 2.5 2.5>
      "Float (Alpha)"=&FloatFromCurve(FloatCurve=DataInterface() {propertyValues {Curve {keys [<at 0.55 1>,<at 1 0>]}}})
     )
    )
    R3 Purple:LinearColor=&MultiplyLinearColor(
     Linear Color=<rgba 0.3 0.25 0.35 1>
     Scale Factor=&MakeVector4FromVectorAndFloat(
      "Vector (RGB)"=<2.5 2.5 2.5>
      "Float (Alpha)"=&FloatFromCurve(FloatCurve=DataInterface() {propertyValues {Curve {keys [<at 0.55 1>,<at 1 0>]}}})
     )
    )
    R3 Saffron:LinearColor=&MultiplyLinearColor(
     Linear Color=<rgba 0.6 0.6 0.1 1>
     Scale Factor=&MakeVector4FromVectorAndFloat(
      "Vector (RGB)"=<2.55 2.55 2.55>
      "Float (Alpha)"=&FloatFromCurve(FloatCurve=DataInterface() {propertyValues {Curve {keys [<at 0.55 1>,<at 1 0>]}}})
     )
    )
    R3 SkyBlue:LinearColor=&MultiplyLinearColor(
     Linear Color=<rgba 0.35 0.6 0.8 1>
     Scale Factor=&MakeVector4FromVectorAndFloat(
      "Vector (RGB)"=<2.35 2.35 2.35>
      "Float (Alpha)"=&FloatFromCurve(FloatCurve=DataInterface() {propertyValues {Curve {keys [<at 0.55 1>,<at 1 0>]}}})
     )
    )
    R3 TitaniumWhite:LinearColor=&MultiplyLinearColor(
     Linear Color=<rgba 1 1 1 1>
     Scale Factor=&MakeVector4FromVectorAndFloat(
      "Vector (RGB)"=<1.2 1.2 1.2>
      "Float (Alpha)"=&FloatFromCurve(FloatCurve=DataInterface() {propertyValues {Curve {keys [<at 0.55 1>,<at 1 0>]}}})
     )
    )
    Saffron=&MultiplyLinearColor(
     Linear Color=<rgba 0.5 0.4 0.1 1>
     Scale Factor=&MakeVector4FromVectorAndFloat("Vector (RGB)"=&VectorFromFloat(Value=0.55), "Float (Alpha)"=2.2)
    )
    SkyBlue=&MultiplyLinearColor(
     Linear Color=<rgba 0.05 0.5 0.8 1>
     Scale Factor=&MakeVector4FromVectorAndFloat("Vector (RGB)"=&VectorFromFloat(Value=0.4), "Float (Alpha)"=2.2)
    )
    TitaniumWhite=&MultiplyLinearColor(
     Linear Color=<rgba 1 1 1.05 1>
     Scale Factor=&MakeVector4FromVectorAndFloat("Vector (RGB)"=&VectorFromFloat(Value=0.4), "Float (Alpha)"=1.7)
    )
    Use R1 Painted Overrides=true
    Use R3 Painted Overrides=true
   ) {module=VC_PaintedVariantsOverride001}
   &FN_DisableTireFX(
    Mask Index Attribute:NiagaraInt32=$Particles.PartitionParticles.Partition
    Mask Index FL=2
    Mask Index FR=3
    Mask Index RL=0
    Mask Index RR=1
   )
   &ParticleState()
   &SpriteRotationRate(Rotation Rate=50) {enabled=false}
   &InheritSourceMovement(
    Applied Position Falloff Curve=DataInterface() {propertyValues {Curve {
       defaultValue=1
       keys [{interpMode=RCIM_Cubic, tangentMode=RCTM_User, time=0, value=1}, {interpMode=RCIM_Cubic, tangentMode=RCTM_User, time=1, value=1}]
      }}}
    Applied Position Falloff Distance=0.1
    Applied Position Scale=&VectorFromFloat(Value=1)
    ApplyForce=false
    ApplyPosition=true
    ApplyPositionFalloff=true
    ApplyPositionFalloffCurve=true
   ) {enabled=false}
   &ScaleSpriteSize(
    Scale Sprite Size Mode=::NewEnumerator0
    Uniform Scale Factor=&Multiply_Float(
     A=3
     B=&MaskFloatBySpawnGroup(
      Masked Float=&FloatFromCurve(FloatCurve=DataInterface() {propertyValues {Curve {
          defaultValue=1
          keys [{interpMode=RCIM_Cubic, tangentMode=RCTM_User, time=0.25, value=1}, {interpMode=RCIM_Cubic, tangentMode=RCTM_User, time=1, value=0.75}]
         }}})
      Passthrough Float=&MaskFloatBySpawnGroup(
       Masked Float=1
       Passthrough Float=&MaskFloatBySpawnGroup(
        Masked Float=&FloatFromCurve(FloatCurve=DataInterface() {propertyValues {Curve {defaultValue=1, keys [{tangentMode=RCTM_User, time=0, value=0}, {tangentMode=RCTM_User, time=0.75, value=1}]}}})
        Passthrough Float=&FloatFromCurve(FloatCurve=DataInterface() {propertyValues {Curve {keys [<at 0.1 1 RCIM_Cubic>,<at 1 0>]}}})
        Spawn Group Mask=2
       )
       Spawn Group Mask=1
      )
     )
    )
   )
   &GravityForce(
    Coordinate Space=::Local
    Gravity=&SelectVectorFromArray(
     Direct Array Index=&SpawnGroup()
     Vector Selection Array=DataInterface() {propertyValues {FloatData [<0 0 -200>, <-500 0 -250>, <-2500 0 0>, <-2200 0 -1800>]}}
    )
   ) {enabled=false}
   &SubUVAnimation(
    Evaluation Type=::NewEnumerator0
    "Sprite Renderer "=DataInterface() {propertyValues {SpriteRenderer=&self:NE_VC_DriftTrail_Cone_0.NiagaraSpriteRendererProperties_6}}
    SubUV Animation Mode:ENiagara_SubUVLookupModeV2=::NewEnumerator2
    UseEndFrame=false
    UseStartFrame=false
   )
   &ScaleVelocity(Velocity Scale=&VectorFromCurve(VectorCurve=DataInterface() {propertyValues {
       XCurve {
        defaultValue=1
        keys [{interpMode=RCIM_Cubic, tangentMode=RCTM_User, time=0, value=0.5}, {interpMode=RCIM_Cubic, tangentMode=RCTM_User, time=1, value=1}]
       }
       YCurve {
        defaultValue=1
        keys [{interpMode=RCIM_Cubic, tangentMode=RCTM_User, time=0, value=0.5}, {interpMode=RCIM_Cubic, tangentMode=RCTM_User, time=1, value=1}]
       }
       ZCurve {defaultValue=1, keys [{tangentMode=RCTM_User, time=0, value=1}, {tangentMode=RCTM_User, time=1, value=1}]}
      }})) {enabled=false}
   &SolveForcesAndVelocity(Clamp Velocity=false, Limit Acceleration=false)
   &DynamicMaterialParameters(
    Index 0 Param 1=&MaskFloatBySpawnGroup(
     Masked Float=&FloatFromCurve(FloatCurve=DataInterface() {propertyValues {Curve {
         defaultValue=1
         keys [{interpMode=RCIM_Cubic, tangentMode=RCTM_User, time=0.35, value=0}, {interpMode=RCIM_Cubic, tangentMode=RCTM_User, time=1, value=1}]
        }}})
     Passthrough Float=&MaskFloatBySpawnGroup(
      Masked Float=&FloatFromCurve(FloatCurve=DataInterface() {propertyValues {Curve {
          defaultValue=1
          keys [{interpMode=RCIM_Cubic, tangentMode=RCTM_User, time=0.5, value=0}, {interpMode=RCIM_Cubic, tangentMode=RCTM_User, time=1, value=1}]
         }}})
      Passthrough Float=&MaskFloatBySpawnGroup(Masked Float=1, Passthrough Float=1, Spawn Group Mask=2)
      Spawn Group Mask=1
     )
    )
    Module.Index0Float0Write=true
    Module.Index0Float1Write=true
    Module.Index0Float2Write=true
    Module.Index0Float3Write=true
    Module.Index1Float0Write=true
    Module.Index1Float1Write=true
    Module.Index1Float2Write=true
    Module.Index1Float3Write=true
    Write Parameter Index 1=::NewEnumerator1
   ) {module=DynamicMaterialParameters001}
  ]
  rendererClasses [&NiagaraSpriteRendererProperties]
  renderers [
   {rendererClass=&NiagaraSpriteRendererProperties}
   {rendererClass=&NiagaraSpriteRendererProperties, rendererIndex=1}
   {rendererClass=&NiagaraSpriteRendererProperties, rendererIndex=2}
   {rendererClass=&NiagaraSpriteRendererProperties, rendererIndex=3}
  ]
 }
 NE_VC_DriftTrail_ConeClouds001 {
  emitterUpdateScript [
   &EmitterState()
   &SpawnRate(SpawnRate=100, Use Spawn Probability=false)
   &SpawnRate(SpawnGroup=1, SpawnRate=100, Use Spawn Probability=true) {module=SpawnRate002}
  ]
  particleSpawnScript [
   &VC_PaintedVariantsOverride(DefaultColor=<rgba 0 0 0 3>, R1 DefaultColor=<rgba 0.4233 0.3549 0.2822 1>, Use R1 Painted Overrides=true) {enabled=false}
   /SetVariables(Particles.VisibilityTag=$Emitter.SpawnGroup) {isSetParameters=true, module=SetVariables_#4}
   &InitializeParticle(
    Color Mode=::NewEnumerator0
    Lifetime Max=&SelectFloatFromArray(
     Direct Array Index=&SpawnGroup()
     Float Selection Array=DataInterface() {propertyValues {FloatData [0.25,0.475]}}
    )
    Lifetime Min:NiagaraFloat=&SelectFloatFromArray(
     Direct Array Index=&SpawnGroup()
     Float Selection Array=DataInterface() {propertyValues {FloatData [0.15,0.25]}}
    )
    Lifetime Mode=::NewEnumerator1
    Position Offset=<-50 0 25>
    Sprite Rotation Mode=::NewEnumerator2
    Sprite Size Max=&SelectVector2DFromArray(
     Direct Array Index=&SpawnGroup()
     Vector 2D Selection Array=DataInterface() {propertyValues {FloatData [<17 17>, <18 18>]}}
    )
    Sprite Size Min=&SelectVector2DFromArray(
     Direct Array Index=&SpawnGroup()
     Vector 2D Selection Array=DataInterface() {propertyValues {FloatData [<8 8>, <10 10>]}}
    )
    Sprite Size Mode=::NewEnumerator2
    UsePositionOffset=false
    Write Material Random=true
   )
   &PartitionParticles()
   &SkeletalMeshLocation(
    Bone / Socket Index=&Add_Integer(A=$Particles.PartitionParticles.Partition, B=4)
    Bone / Socket Sampling Mode=::NewEnumerator5
    Interpolate Mesh Sampling=false
    MaskSpawnGroup=false
    Mesh Sampling Type=::NewEnumerator6
    Skeletal Mesh=$System.SkeletalMesh
   )
   &AddVelocity(
    Cone Angle=&MaskFloatBySpawnGroup(Masked Float=60, Passthrough Float=1)
    Cone Axis=&RandomRangeVector(
     Minimum=&SelectVectorFromArray(
      Direct Array Index=&SpawnGroup()
      Vector Selection Array=DataInterface() {propertyValues {FloatData [<-1 -0.25 0.3>, <-1 0 0>]}}
     )
     Maximum=&SelectVectorFromArray(
      Direct Array Index=&SpawnGroup()
      Vector Selection Array=DataInterface() {propertyValues {FloatData [<-0.35 0.25 1>, <-1 0 0>]}}
     )
    )
    Module.UseConeFalloff=false
    Module.UseVelDistribution=false
    Override Seed=false
    Velocity Mode=::NewEnumerator2
    Velocity Speed=&MergeHelper_FloatSwitch(
     0=&RandomRangeFloat(
      Minimum=&SelectFloatFromArray(Direct Array Index=&SpawnGroup(), Float Selection Array=DataInterface() {propertyValues {FloatData [100,150]}})
      Maximum=&SelectFloatFromArray(Direct Array Index=&SpawnGroup(), Float Selection Array=DataInterface() {propertyValues {FloatData [300,200]}})
     )
     1=&RandomRangeFloat(
      Minimum=&SelectFloatFromArray(Direct Array Index=&SpawnGroup(), Float Selection Array=DataInterface() {propertyValues {FloatData [100,100]}})
      Maximum=&SelectFloatFromArray(Direct Array Index=&SpawnGroup(), Float Selection Array=DataInterface() {propertyValues {FloatData [300,150]}})
     )
     2=&RandomRangeFloat(
      Minimum=&SelectFloatFromArray(Direct Array Index=&SpawnGroup(), Float Selection Array=DataInterface() {propertyValues {FloatData [100,150]}})
      Maximum=&SelectFloatFromArray(Direct Array Index=&SpawnGroup(), Float Selection Array=DataInterface() {propertyValues {FloatData [300,200]}})
     )
     3=&RandomRangeFloat(
      Minimum=&SelectFloatFromArray(Direct Array Index=&SpawnGroup(), Float Selection Array=DataInterface() {propertyValues {FloatData [100,75]}})
      Maximum=&SelectFloatFromArray(Direct Array Index=&SpawnGroup(), Float Selection Array=DataInterface() {propertyValues {FloatData [300,125]}})
     )
     Index=$System.GameModeID
     Num=::NewEnumerator2
    )
   )
   &ShapeLocation(
    Hemisphere Distribution=<1 0.75>
    MaskSpawnGroup=false
    Offset=<-65 0 -24>
    Offset Coordinate Space=::Local
    Offset Mode=::NewEnumerator0
    Override Seed=false
    Sphere Radius=&SelectFloatFromArray(Direct Array Index=&SpawnGroup(), Float Selection Array=DataInterface() {propertyValues {FloatData [20,14]}})
    Sphere Surface Distribution=0.1
    UseOwnerScale=true
   )
  ]
  particleUpdateScript [
   &VC_PaintedVariantsOverride(
    Black=&MultiplyLinearColor(
     Linear Color=<rgba 0 0 0 2.75>
     Scale Factor=&MakeVector4FromVectorAndFloat(
      "Vector (RGB)"=<1 1 1>
      "Float (Alpha)"=&FloatFromCurve(FloatCurve=DataInterface() {propertyValues {Curve {
          defaultValue=1
          keys [{interpMode=RCIM_Cubic, tangentMode=RCTM_User, time=0.25, value=1}, {interpMode=RCIM_Cubic, tangentMode=RCTM_User, time=1, value=0}]
         }}})
     )
    )
    BurntSienna=&MultiplyLinearColor(
     Linear Color=<rgba 0.1512 0.0714 0.0285 2.25>
     Scale Factor=&MakeVector4FromVectorAndFloat(
      "Vector (RGB)"=<1 1 1>
      "Float (Alpha)"=&FloatFromCurve(
       FloatCurve=DataInterface() {propertyValues {Curve {
          defaultValue=1
          keys [{interpMode=RCIM_Cubic, tangentMode=RCTM_User, time=0.25, value=1}, {interpMode=RCIM_Cubic, tangentMode=RCTM_User, time=1, value=0}]
         }}}
       Scale Curve=1.25
      )
     )
    )
    Cobalt=&MultiplyLinearColor(
     Linear Color=<rgba 0.05 0.1 1 2.4>
     Scale Factor=&MakeVector4FromVectorAndFloat(
      "Vector (RGB)"=<0.2 0.2 0.2>
      "Float (Alpha)"=&FloatFromCurve(FloatCurve=DataInterface() {propertyValues {Curve {
          defaultValue=1
          keys [{interpMode=RCIM_Cubic, tangentMode=RCTM_User, time=0.25, value=1}, {interpMode=RCIM_Cubic, tangentMode=RCTM_User, time=1, value=0}]
         }}})
     )
    )
    Crimson=&MultiplyLinearColor(
     Linear Color=<rgba 0.5 0 0 2.4>
     Scale Factor=&MakeVector4FromVectorAndFloat(
      "Vector (RGB)"=<0.5 0.5 0.5>
      "Float (Alpha)"=&FloatFromCurve(FloatCurve=DataInterface() {propertyValues {Curve {
          defaultValue=1
          keys [{interpMode=RCIM_Cubic, tangentMode=RCTM_User, time=0.25, value=1}, {interpMode=RCIM_Cubic, tangentMode=RCTM_User, time=1, value=0}]
         }}})
     )
    )
    DefaultColor=&MultiplyLinearColor(
     Linear Color=<rgba 0 0 0 2.75>
     Scale Factor=&MakeVector4FromVectorAndFloat(
      "Vector (RGB)"=<1 1 1>
      "Float (Alpha)"=&FloatFromCurve(FloatCurve=DataInterface() {propertyValues {Curve {
          defaultValue=1
          keys [{interpMode=RCIM_Cubic, tangentMode=RCTM_User, time=0.25, value=1}, {interpMode=RCIM_Cubic, tangentMode=RCTM_User, time=1, value=0}]
         }}})
     )
    )
    ForestGreen=&MultiplyLinearColor(
     Linear Color=<rgba 0 0.35 0 2.4>
     Scale Factor=&MakeVector4FromVectorAndFloat(
      "Vector (RGB)"=<0.5 0.5 0.5>
      "Float (Alpha)"=&FloatFromCurve(FloatCurve=DataInterface() {propertyValues {Curve {
          defaultValue=1
          keys [{interpMode=RCIM_Cubic, tangentMode=RCTM_User, time=0.25, value=1}, {interpMode=RCIM_Cubic, tangentMode=RCTM_User, time=1, value=0}]
         }}})
     )
    )
    Grey=&MultiplyLinearColor(
     Linear Color=<rgba 0.125 0.125 0.14 2.4>
     Scale Factor=&MakeVector4FromVectorAndFloat(
      "Vector (RGB)"=<1.25 1.25 1.25>
      "Float (Alpha)"=&FloatFromCurve(FloatCurve=DataInterface() {propertyValues {Curve {
          defaultValue=1
          keys [{interpMode=RCIM_Cubic, tangentMode=RCTM_User, time=0.25, value=1}, {interpMode=RCIM_Cubic, tangentMode=RCTM_User, time=1, value=0}]
         }}})
     )
    )
    Lime=&MultiplyLinearColor(
     Linear Color=<rgba 0.3 0.7 0 2.1>
     Scale Factor=&MakeVector4FromVectorAndFloat(
      "Vector (RGB)"=<0.35 0.35 0.35>
      "Float (Alpha)"=&FloatFromCurve(FloatCurve=DataInterface() {propertyValues {Curve {
          defaultValue=1
          keys [{interpMode=RCIM_Cubic, tangentMode=RCTM_User, time=0.25, value=1}, {interpMode=RCIM_Cubic, tangentMode=RCTM_User, time=1, value=0}]
         }}})
     )
    )
    Orange=&MultiplyLinearColor(
     Linear Color=<rgba 0.7 0.35 0.01 2.2>
     Scale Factor=&MakeVector4FromVectorAndFloat(
      "Vector (RGB)"=<0.5 0.5 0.5>
      "Float (Alpha)"=&FloatFromCurve(FloatCurve=DataInterface() {propertyValues {Curve {
          defaultValue=1
          keys [{interpMode=RCIM_Cubic, tangentMode=RCTM_User, time=0.25, value=1}, {interpMode=RCIM_Cubic, tangentMode=RCTM_User, time=1, value=0}]
         }}})
     )
    )
    Pink=&MultiplyLinearColor(
     Linear Color=<rgba 0.55 0.145 0.4 2.2>
     Scale Factor=&MakeVector4FromVectorAndFloat(
      "Vector (RGB)"=<0.4 0.4 0.4>
      "Float (Alpha)"=&FloatFromCurve(FloatCurve=DataInterface() {propertyValues {Curve {
          defaultValue=1
          keys [{interpMode=RCIM_Cubic, tangentMode=RCTM_User, time=0.25, value=1}, {interpMode=RCIM_Cubic, tangentMode=RCTM_User, time=1, value=0}]
         }}})
     )
    )
    Purple=&MultiplyLinearColor(
     Linear Color=<rgba 0.25 0 0.3 2.4>
     Scale Factor=&MakeVector4FromVectorAndFloat(
      "Vector (RGB)"=<0.4 0.4 0.4>
      "Float (Alpha)"=&FloatFromCurve(FloatCurve=DataInterface() {propertyValues {Curve {
          defaultValue=1
          keys [{interpMode=RCIM_Cubic, tangentMode=RCTM_User, time=0.25, value=1}, {interpMode=RCIM_Cubic, tangentMode=RCTM_User, time=1, value=0}]
         }}})
     )
    )
    R2 Black:LinearColor=&MultiplyLinearColor(
     Linear Color=<rgba 0.11 0.11 0.11 0.98>
     Scale Factor=&MakeVector4FromVectorAndFloat(
      "Vector (RGB)"=&VectorFromCurve(VectorCurve=DataInterface() {propertyValues {
         XCurve {
          defaultValue=1
          keys [{interpMode=RCIM_Cubic, tangentMode=RCTM_User, time=0, value=0}, {interpMode=RCIM_Cubic, tangentMode=RCTM_User, time=0.44, value=1}]
         }
         YCurve {
          defaultValue=1
          keys [{interpMode=RCIM_Cubic, tangentMode=RCTM_User, time=0, value=0}, {interpMode=RCIM_Cubic, tangentMode=RCTM_User, time=0.44, value=1}]
         }
         ZCurve {
          defaultValue=1
          keys [{interpMode=RCIM_Cubic, tangentMode=RCTM_User, time=0, value=0}, {interpMode=RCIM_Cubic, tangentMode=RCTM_User, time=0.44, value=1}]
         }
        }})
      "Float (Alpha)"=0.95
     )
    )
    R2 BurntSienna:LinearColor=&MultiplyLinearColor(
     Linear Color=<rgba 0.143 0.066 0.0275 0.98>
     Scale Factor=&MakeVector4FromVectorAndFloat(
      "Vector (RGB)"=&VectorFromCurve(VectorCurve=DataInterface() {propertyValues {
         XCurve {
          defaultValue=1
          keys [{interpMode=RCIM_Cubic, tangentMode=RCTM_User, time=0, value=0}, {interpMode=RCIM_Cubic, tangentMode=RCTM_User, time=0.44, value=1}]
         }
         YCurve {
          defaultValue=1
          keys [{interpMode=RCIM_Cubic, tangentMode=RCTM_User, time=0, value=0}, {interpMode=RCIM_Cubic, tangentMode=RCTM_User, time=0.44, value=1}]
         }
         ZCurve {
          defaultValue=1
          keys [{interpMode=RCIM_Cubic, tangentMode=RCTM_User, time=0, value=0}, {interpMode=RCIM_Cubic, tangentMode=RCTM_User, time=0.44, value=1}]
         }
        }})
      "Float (Alpha)"=0.95
     )
    )
    R2 Cobalt:LinearColor=&MultiplyLinearColor(
     Linear Color=<rgba 0.15 0.3 0.9 0.98>
     Scale Factor=&MakeVector4FromVectorAndFloat(
      "Vector (RGB)"=&VectorFromCurve(
       VectorCurve=DataInterface() {propertyValues {
         XCurve {
          defaultValue=1
          keys [{interpMode=RCIM_Cubic, tangentMode=RCTM_User, time=0, value=0.75}, {interpMode=RCIM_Cubic, tangentMode=RCTM_User, time=0.44, value=1}]
         }
         YCurve {
          defaultValue=1
          keys [{interpMode=RCIM_Cubic, tangentMode=RCTM_User, time=0, value=0.75}, {interpMode=RCIM_Cubic, tangentMode=RCTM_User, time=0.44, value=1}]
         }
         ZCurve {
          defaultValue=1
          keys [{interpMode=RCIM_Cubic, tangentMode=RCTM_User, time=0, value=0.75}, {interpMode=RCIM_Cubic, tangentMode=RCTM_User, time=0.44, value=1}]
         }
        }}
       Scale Curve=<0.85 0.85 0.85>
      )
      "Float (Alpha)"=0.95
     )
    )
    R2 Crimson:LinearColor=&MultiplyLinearColor(
     Linear Color=<rgba 0.45 0.15 0.15 0.98>
     Scale Factor=&MakeVector4FromVectorAndFloat(
      "Vector (RGB)"=&VectorFromCurve(
       VectorCurve=DataInterface() {propertyValues {
         XCurve {
          defaultValue=1
          keys [{interpMode=RCIM_Cubic, tangentMode=RCTM_User, time=0, value=0.35}, {interpMode=RCIM_Cubic, tangentMode=RCTM_User, time=0.44, value=1}]
         }
         YCurve {
          defaultValue=1
          keys [{interpMode=RCIM_Cubic, tangentMode=RCTM_User, time=0, value=0.35}, {interpMode=RCIM_Cubic, tangentMode=RCTM_User, time=0.44, value=1}]
         }
         ZCurve {
          defaultValue=1
          keys [{interpMode=RCIM_Cubic, tangentMode=RCTM_User, time=0, value=0.35}, {interpMode=RCIM_Cubic, tangentMode=RCTM_User, time=0.44, value=1}]
         }
        }}
       Scale Curve=<0.85 0.85 0.85>
      )
      "Float (Alpha)"=0.95
     )
    )
    R2 DefaultColor:LinearColor=&MultiplyLinearColor(
     Linear Color=<rgba 0.13 0.044 0.0174 0.98>
     Scale Factor=&MakeVector4FromVectorAndFloat(
      "Vector (RGB)"=&VectorFromCurve(
       VectorCurve=DataInterface() {propertyValues {
         XCurve {
          defaultValue=1
          keys [{interpMode=RCIM_Cubic, tangentMode=RCTM_User, time=0, value=0}, {interpMode=RCIM_Cubic, tangentMode=RCTM_User, time=0.44, value=1}]
         }
         YCurve {
          defaultValue=1
          keys [{interpMode=RCIM_Cubic, tangentMode=RCTM_User, time=0, value=0}, {interpMode=RCIM_Cubic, tangentMode=RCTM_User, time=0.44, value=1}]
         }
         ZCurve {
          defaultValue=1
          keys [{interpMode=RCIM_Cubic, tangentMode=RCTM_User, time=0, value=0}, {interpMode=RCIM_Cubic, tangentMode=RCTM_User, time=0.44, value=1}]
         }
        }}
       Scale Curve=<1.5 2.5 1.5>
      )
      "Float (Alpha)"=0.95
     )
    )
    R2 ForestGreen:LinearColor=&MultiplyLinearColor(
     Linear Color=<rgba 0.15 0.35 0.15 0.98>
     Scale Factor=&MakeVector4FromVectorAndFloat(
      "Vector (RGB)"=&VectorFromCurve(
       VectorCurve=DataInterface() {propertyValues {
         XCurve {
          defaultValue=1
          keys [{interpMode=RCIM_Cubic, tangentMode=RCTM_User, time=0, value=0.35}, {interpMode=RCIM_Cubic, tangentMode=RCTM_User, time=0.44, value=1}]
         }
         YCurve {
          defaultValue=1
          keys [{interpMode=RCIM_Cubic, tangentMode=RCTM_User, time=0, value=0.35}, {interpMode=RCIM_Cubic, tangentMode=RCTM_User, time=0.44, value=1}]
         }
         ZCurve {
          defaultValue=1
          keys [{interpMode=RCIM_Cubic, tangentMode=RCTM_User, time=0, value=0.35}, {interpMode=RCIM_Cubic, tangentMode=RCTM_User, time=0.44, value=1}]
         }
        }}
       Scale Curve=<0.85 0.85 0.85>
      )
      "Float (Alpha)"=0.95
     )
    )
    R2 Grey:LinearColor=&MultiplyLinearColor(
     Linear Color=<rgba 0.33 0.33 0.4 0.98>
     Scale Factor=&MakeVector4FromVectorAndFloat(
      "Vector (RGB)"=&VectorFromCurve(
       VectorCurve=DataInterface() {propertyValues {
         XCurve {
          defaultValue=1
          keys [{interpMode=RCIM_Cubic, tangentMode=RCTM_User, time=0, value=0.4}, {interpMode=RCIM_Cubic, tangentMode=RCTM_User, time=0.44, value=1}]
         }
         YCurve {
          defaultValue=1
          keys [{interpMode=RCIM_Cubic, tangentMode=RCTM_User, time=0, value=0.4}, {interpMode=RCIM_Cubic, tangentMode=RCTM_User, time=0.44, value=1}]
         }
         ZCurve {
          defaultValue=1
          keys [{interpMode=RCIM_Cubic, tangentMode=RCTM_User, time=0, value=0.4}, {interpMode=RCIM_Cubic, tangentMode=RCTM_User, time=0.44, value=1}]
         }
        }}
       Scale Curve=<0.85 0.85 0.85>
      )
      "Float (Alpha)"=0.9
     )
    )
    R2 Lime:LinearColor=&MultiplyLinearColor(
     Linear Color=<rgba 0.4 0.6 0 0.98>
     Scale Factor=&MakeVector4FromVectorAndFloat(
      "Vector (RGB)"=&VectorFromCurve(
       VectorCurve=DataInterface() {propertyValues {
         XCurve {
          defaultValue=1
          keys [{interpMode=RCIM_Cubic, tangentMode=RCTM_User, time=0, value=0.45}, {interpMode=RCIM_Cubic, tangentMode=RCTM_User, time=0.44, value=1}]
         }
         YCurve {
          defaultValue=1
          keys [{interpMode=RCIM_Cubic, tangentMode=RCTM_User, time=0, value=0.45}, {interpMode=RCIM_Cubic, tangentMode=RCTM_User, time=0.44, value=1}]
         }
         ZCurve {
          defaultValue=1
          keys [{interpMode=RCIM_Cubic, tangentMode=RCTM_User, time=0, value=0.45}, {interpMode=RCIM_Cubic, tangentMode=RCTM_User, time=0.44, value=1}]
         }
        }}
       Scale Curve=<0.85 0.85 0.85>
      )
      "Float (Alpha)"=0.95
     )
    )
    R2 Orange:LinearColor=&MultiplyLinearColor(
     Linear Color=<rgba 0.7 0.3 0.2 0.98>
     Scale Factor=&MakeVector4FromVectorAndFloat(
      "Vector (RGB)"=&VectorFromCurve(
       VectorCurve=DataInterface() {propertyValues {
         XCurve {
          defaultValue=1
          keys [{interpMode=RCIM_Cubic, tangentMode=RCTM_User, time=0, value=0.15}, {interpMode=RCIM_Cubic, tangentMode=RCTM_User, time=0.44, value=1}]
         }
         YCurve {
          defaultValue=1
          keys [{interpMode=RCIM_Cubic, tangentMode=RCTM_User, time=0, value=0.15}, {interpMode=RCIM_Cubic, tangentMode=RCTM_User, time=0.44, value=1}]
         }
         ZCurve {
          defaultValue=1
          keys [{interpMode=RCIM_Cubic, tangentMode=RCTM_User, time=0, value=0.15}, {interpMode=RCIM_Cubic, tangentMode=RCTM_User, time=0.44, value=1}]
         }
        }}
       Scale Curve=<0.85 0.85 0.85>
      )
      "Float (Alpha)"=0.95
     )
    )
    R2 Pink:LinearColor=&MultiplyLinearColor(
     Linear Color=<rgba 0.7 0.28 0.5982 1>
     Scale Factor=&MakeVector4FromVectorAndFloat(
      "Vector (RGB)"=&VectorFromCurve(
       VectorCurve=DataInterface() {propertyValues {
         XCurve {
          defaultValue=1
          keys [{interpMode=RCIM_Cubic, tangentMode=RCTM_User, time=0, value=0.3}, {interpMode=RCIM_Cubic, tangentMode=RCTM_User, time=0.44, value=1}]
         }
         YCurve {
          defaultValue=1
          keys [{interpMode=RCIM_Cubic, tangentMode=RCTM_User, time=0, value=0.3}, {interpMode=RCIM_Cubic, tangentMode=RCTM_User, time=0.44, value=1}]
         }
         ZCurve {
          defaultValue=1
          keys [{interpMode=RCIM_Cubic, tangentMode=RCTM_User, time=0, value=0.3}, {interpMode=RCIM_Cubic, tangentMode=RCTM_User, time=0.44, value=1}]
         }
        }}
       Scale Curve=<0.85 0.85 0.85>
      )
      "Float (Alpha)"=0.95
     )
    )
    R2 Purple:LinearColor=&MultiplyLinearColor(
     Linear Color=<rgba 0.3203 0.1838 0.525 1>
     Scale Factor=&MakeVector4FromVectorAndFloat(
      "Vector (RGB)"=&VectorFromCurve(
       VectorCurve=DataInterface() {propertyValues {
         XCurve {
          defaultValue=1
          keys [{interpMode=RCIM_Cubic, tangentMode=RCTM_User, time=0, value=0.25}, {interpMode=RCIM_Cubic, tangentMode=RCTM_User, time=0.44, value=1}]
         }
         YCurve {
          defaultValue=1
          keys [{interpMode=RCIM_Cubic, tangentMode=RCTM_User, time=0, value=0.25}, {interpMode=RCIM_Cubic, tangentMode=RCTM_User, time=0.44, value=1}]
         }
         ZCurve {
          defaultValue=1
          keys [{interpMode=RCIM_Cubic, tangentMode=RCTM_User, time=0, value=0.25}, {interpMode=RCIM_Cubic, tangentMode=RCTM_User, time=0.44, value=1}]
         }
        }}
       Scale Curve=<0.85 0.85 0.85>
      )
      "Float (Alpha)"=0.95
     )
    )
    R2 Saffron:LinearColor=&MultiplyLinearColor(
     Linear Color=<rgba 0.9 0.7861 0.09 0.98>
     Scale Factor=&MakeVector4FromVectorAndFloat(
      "Vector (RGB)"=&VectorFromCurve(
       VectorCurve=DataInterface() {propertyValues {
         XCurve {
          defaultValue=1
          keys [{interpMode=RCIM_Cubic, tangentMode=RCTM_User, time=0, value=0.3}, {interpMode=RCIM_Cubic, tangentMode=RCTM_User, time=0.44, value=1}]
         }
         YCurve {
          defaultValue=1
          keys [{interpMode=RCIM_Cubic, tangentMode=RCTM_User, time=0, value=0.3}, {interpMode=RCIM_Cubic, tangentMode=RCTM_User, time=0.44, value=1}]
         }
         ZCurve {
          defaultValue=1
          keys [{interpMode=RCIM_Cubic, tangentMode=RCTM_User, time=0, value=0.3}, {interpMode=RCIM_Cubic, tangentMode=RCTM_User, time=0.44, value=1}]
         }
        }}
       Scale Curve=<0.85 0.85 0.85>
      )
      "Float (Alpha)"=0.9
     )
    )
    R2 SkyBlue:LinearColor=&MultiplyLinearColor(
     Linear Color=<rgba 0.1125 0.5625 0.9 0.98>
     Scale Factor=&MakeVector4FromVectorAndFloat(
      "Vector (RGB)"=&VectorFromCurve(
       VectorCurve=DataInterface() {propertyValues {
         XCurve {
          defaultValue=1
          keys [{interpMode=RCIM_Cubic, tangentMode=RCTM_User, time=0, value=0.3}, {interpMode=RCIM_Cubic, tangentMode=RCTM_User, time=0.44, value=1}]
         }
         YCurve {
          defaultValue=1
          keys [{interpMode=RCIM_Cubic, tangentMode=RCTM_User, time=0, value=0.3}, {interpMode=RCIM_Cubic, tangentMode=RCTM_User, time=0.44, value=1}]
         }
         ZCurve {
          defaultValue=1
          keys [{interpMode=RCIM_Cubic, tangentMode=RCTM_User, time=0, value=0.3}, {interpMode=RCIM_Cubic, tangentMode=RCTM_User, time=0.44, value=1}]
         }
        }}
       Scale Curve=<0.85 0.85 0.85>
      )
      "Float (Alpha)"=0.95
     )
    )
    R2 TitaniumWhite:LinearColor=&MultiplyLinearColor(
     Linear Color=<rgba 1 1 1 0.98>
     Scale Factor=&MakeVector4FromVectorAndFloat(
      "Vector (RGB)"=&VectorFromCurve(
       VectorCurve=DataInterface() {propertyValues {
         XCurve {
          defaultValue=1
          keys [{interpMode=RCIM_Cubic, tangentMode=RCTM_User, time=0, value=0.3}, {interpMode=RCIM_Cubic, tangentMode=RCTM_User, time=0.44, value=1}]
         }
         YCurve {
          defaultValue=1
          keys [{interpMode=RCIM_Cubic, tangentMode=RCTM_User, time=0, value=0.3}, {interpMode=RCIM_Cubic, tangentMode=RCTM_User, time=0.44, value=1}]
         }
         ZCurve {
          defaultValue=1
          keys [{interpMode=RCIM_Cubic, tangentMode=RCTM_User, time=0, value=0.3}, {interpMode=RCIM_Cubic, tangentMode=RCTM_User, time=0.44, value=1}]
         }
        }}
       Scale Curve=<0.7 0.7 0.75>
      )
      "Float (Alpha)"=0.9
     )
    )
    Saffron=&MultiplyLinearColor(
     Linear Color=<rgba 0.5 0.4 0 2.6>
     Scale Factor=&MakeVector4FromVectorAndFloat(
      "Vector (RGB)"=<0.45 0.45 0.45>
      "Float (Alpha)"=&FloatFromCurve(FloatCurve=DataInterface() {propertyValues {Curve {
          defaultValue=1
          keys [{interpMode=RCIM_Cubic, tangentMode=RCTM_User, time=0.25, value=1}, {interpMode=RCIM_Cubic, tangentMode=RCTM_User, time=1, value=0}]
         }}})
     )
    )
    SkyBlue=&MultiplyLinearColor(
     Linear Color=<rgba 0.05 0.5 0.8 2.3>
     Scale Factor=&MakeVector4FromVectorAndFloat(
      "Vector (RGB)"=<0.2 0.2 0.2>
      "Float (Alpha)"=&FloatFromCurve(FloatCurve=DataInterface() {propertyValues {Curve {
          defaultValue=1
          keys [{interpMode=RCIM_Cubic, tangentMode=RCTM_User, time=0.25, value=1}, {interpMode=RCIM_Cubic, tangentMode=RCTM_User, time=1, value=0}]
         }}})
     )
    )
    TitaniumWhite=&MultiplyLinearColor(
     Linear Color=<rgba 1 1 1 1.85>
     Scale Factor=&MakeVector4FromVectorAndFloat(
      "Vector (RGB)"=<0.45 0.45 0.5>
      "Float (Alpha)"=&FloatFromCurve(
       FloatCurve=DataInterface() {propertyValues {Curve {
          defaultValue=1
          keys [{interpMode=RCIM_Cubic, tangentMode=RCTM_User, time=0.25, value=1}, {interpMode=RCIM_Cubic, tangentMode=RCTM_User, time=1, value=0}]
         }}}
       Scale Curve=1.15
      )
     )
    )
    Use R2 Painted Overrides=true
   ) {module=VC_PaintedVariantsOverride001}
   &FN_DisableTireFX(
    Mask Index Attribute:NiagaraInt32=$Particles.PartitionParticles.Partition
    Mask Index FL=2
    Mask Index FR=3
    Mask Index RL=0
    Mask Index RR=1
   )
   &ParticleState()
   &SpriteRotationRate(Rotation Rate=100)
   &InheritSourceMovement(
    Applied Position Falloff Curve=DataInterface() {propertyValues {Curve {
       defaultValue=1
       keys [{interpMode=RCIM_Cubic, tangentMode=RCTM_User, time=0, value=1}, {interpMode=RCIM_Cubic, tangentMode=RCTM_User, time=1, value=1}]
      }}}
    Applied Position Falloff Distance=0.1
    Applied Position Scale=&VectorFromFloat(Value=1)
    ApplyForce=false
    ApplyPosition=true
    ApplyPositionFalloff=true
    ApplyPositionFalloffCurve=true
   ) {enabled=false}
   &ScaleSpriteSize(
    Scale Sprite Size Mode=::NewEnumerator0
    Uniform Scale Factor=&Multiply_Float(
     A=3
     B=&MaskFloatBySpawnGroup(
      Masked Float=&FloatFromCurve(FloatCurve=DataInterface() {propertyValues {Curve {
          defaultValue=1
          keys [{interpMode=RCIM_Cubic, tangentMode=RCTM_User, time=0.25, value=1}, {interpMode=RCIM_Cubic, tangentMode=RCTM_User, time=1, value=0.75}]
         }}})
      Passthrough Float=&FloatFromCurve(
       FloatCurve=DataInterface() {propertyValues {Curve {
          defaultValue=1
          keys [
           {interpMode=RCIM_Cubic, tangentMode=RCTM_User, time=0, value=0.25}
           {interpMode=RCIM_Cubic, tangentMode=RCTM_User, time=0.2, value=1}
           {tangentMode=RCTM_User, time=1, value=0.65}
          ]
         }}}
       Scale Curve=0.825
      )
     )
    )
   )
   &GravityForce(
    Coordinate Space=::Simulation
    Gravity=&SelectVectorFromArray(
     Direct Array Index=&SpawnGroup()
     Vector Selection Array=DataInterface() {propertyValues {FloatData [<-1800 0 -250>, <-400 0 0>]}}
    )
   )
   &SubUVAnimation(
    Evaluation Type=::NewEnumerator0
    "Sprite Renderer "=DataInterface() {propertyValues {SpriteRenderer=&self:NE_VC_DriftTrail_ConeClouds001_0.NiagaraSpriteRendererProperties_6}}
    SubUV Animation Mode:ENiagara_SubUVLookupModeV2=::NewEnumerator2
    UseEndFrame=false
    UseStartFrame=false
   )
   &SolveForcesAndVelocity(Clamp Velocity=false, Limit Acceleration=false)
   &DynamicMaterialParameters(
    Index 0 Param 1=&MaskFloatBySpawnGroup(
     Masked Float=&FloatFromCurve(FloatCurve=DataInterface() {propertyValues {Curve {
         defaultValue=1
         keys [{interpMode=RCIM_Cubic, tangentMode=RCTM_User, time=0.35, value=0.1}, {interpMode=RCIM_Cubic, tangentMode=RCTM_User, time=0.8, value=1}]
        }}})
     Passthrough Float=&FloatFromCurve(FloatCurve=DataInterface() {propertyValues {Curve {
         defaultValue=1
         keys [{interpMode=RCIM_Cubic, tangentMode=RCTM_User, time=0.5, value=0}, {interpMode=RCIM_Cubic, tangentMode=RCTM_User, time=0.8, value=0.4}]
        }}})
    )
    Module.Index0Float0Write=true
    Module.Index0Float1Write=true
    Module.Index0Float2Write=true
    Module.Index0Float3Write=true
    Module.Index1Float0Write=true
    Module.Index1Float1Write=true
    Module.Index1Float2Write=true
    Module.Index1Float3Write=true
    Write Parameter Index 1=::NewEnumerator1
   ) {module=DynamicMaterialParameters001}
  ]
  rendererClasses [&NiagaraSpriteRendererProperties]
  renderers [
   {rendererClass=&NiagaraSpriteRendererProperties}
   {rendererClass=&NiagaraSpriteRendererProperties, rendererIndex=1}
  ]
 }
 NE_VC_DriftTrail_Sprite_Parent001 {
  bEnabled=false
  emitterUpdateScript [
   &EmitterState()
   &SpawnRate(SpawnGroup=1, SpawnRate=18, Use Spawn Probability=false)
   &SpawnRate(SpawnGroup=2, SpawnRate=12, Use Spawn Probability=true) {module=SpawnRate002}
   &SpawnRate(SpawnGroup=3, SpawnRate=12, Use Spawn Probability=false) {module=SpawnRate003}
  ]
  particleSpawnScript [
   &VC_PaintedVariantsOverride() {enabled=false}
   /SetVariables(Particles.VisibilityTag=$Emitter.SpawnGroup) {isSetParameters=true, module=SetVariables_#5}
   &InitializeParticle(
    Lifetime Max=0.8
    Lifetime Min=0.6
    Lifetime Mode=::NewEnumerator1
    Mesh Scale Mode=::NewEnumerator3
    Mesh Uniform Scale=0.1
    Sprite Rotation Mode=::NewEnumerator2
    Sprite Size Mode=::NewEnumerator3
    Uniform Sprite Size=&MaskFloatBySpawnGroup(
     Masked Float=15
     Passthrough Float=&MaskFloatBySpawnGroup(
      Masked Float=&RandomRangeFloat(Minimum=5, Maximum=8)
      Passthrough Float=&MaskFloatBySpawnGroup(
       Masked Float=&RandomRangeFloat(Minimum=5, Maximum=8)
       Passthrough Float=&MaskFloatBySpawnGroup(
        Masked Float=&RandomRangeFloat(Minimum=5, Maximum=8)
        Passthrough Float=&RandomRangeFloat(Minimum=10, Maximum=15)
        Spawn Group Mask=3
       )
       Spawn Group Mask=2
      )
      Spawn Group Mask=1
     )
    )
    UsePositionOffset=false
    Write Material Random=true
   )
   &PartitionParticles()
   &SkeletalMeshLocation(
    Bone / Socket Index=&Add_Integer(A=$Particles.PartitionParticles.Partition, B=4)
    Bone / Socket Sampling Mode=::NewEnumerator5
    MaskSpawnGroup=false
    Mesh Sampling Type=::NewEnumerator6
    Sampled Position Offset=&AddVectorToPosition(Position=<-50 0 -15>, Vector=$User.offset)
    Skeletal Mesh=$System.SkeletalMesh
   )
   &AddVelocity(Velocity=&MaskVectorBySpawnGroup(
     Masked Vector=&RandomRangeVector(Minimum=<-250 0 -5>, Maximum=<-500 0 5>)
     Passthrough Vector=&RandomRangeVector(Minimum=<-300 0 -5>, Maximum=<-600 0 5>)
     Spawn Group Mask=4
    ))
   &InitialMeshOrientation(Use Rotation Vector=false)
   &ShapeLocation(MaskSpawnGroup=false, Override Seed=false, Sphere Radius=20, UseOwnerScale=true)
  ]
  particleUpdateScript [
   &FN_DisableTireFX(
    Mask Index Attribute:NiagaraInt32=$Particles.PartitionParticles.Partition
    Mask Index FL=2
    Mask Index FR=3
    Mask Index RL=0
    Mask Index RR=1
   )
   &ParticleState()
   &SpriteRotationRate(Rotation Rate=50)
   &UpdateMeshOrientation(Rotation Rate=0.5, Rotation Vector=<1 1 1>)
   &InheritSourceMovement(
    Applied Position Falloff Distance=20
    Applied Position Scale=&VectorFromFloat(Value=&FloatFromCurve(FloatCurve=DataInterface() {propertyValues {Curve {keys [{interpMode=RCIM_Cubic, tangentMode=RCTM_Break, time=0.3, value=1}, {interpMode=RCIM_Cubic, tangentMode=RCTM_Break, time=1, value=0}]}}}))
    ApplyForce=false
    ApplyPosition=true
    ApplyPositionFalloff=false
   )
   &ScaleSpriteSize(
    Scale Sprite Size Mode=::NewEnumerator0
    Uniform Scale Factor=&Multiply_Float(
     A=3
     B=&FloatFromCurve(FloatCurve=DataInterface() {propertyValues {Curve {defaultValue=1, keys [<at 0 0.5 RCIM_Cubic>,<at 0.35 1 RCIM_Cubic>]}}})
    )
   )
   /SetVariables(Particles.StarRed:LinearColor=<rgba 0.5104 0 0.002 1>) {isSetParameters=true}
   &CurlNoiseForce(
    Bias Noise=false
    Noise Frequency=50000
    Noise Strength=&RandomRangeFloat(Minimum=100, Maximum=200)
    Pan Noise=false
    Randomization Vector:Vector3f=<0.65 0.125 0>
   )
   &JitterPosition(Jitter Amount:NiagaraFloat=&RandomRangeFloat(Minimum=-5, Maximum=5), Jitter Delay=0.6)
   &SolveForcesAndVelocity(Clamp Velocity=false, Limit Acceleration=false)
   &DynamicMaterialParameters(
    Index 0 Param 1=1
    Module.Index0Float0Write=true
    Module.Index0Float1Write=true
    Module.Index0Float2Write=true
    Module.Index0Float3Write=true
    Module.Index1Float0Write=true
    Module.Index1Float1Write=true
    Module.Index1Float2Write=true
    Module.Index1Float3Write=true
    Write Parameter Index 1=::NewEnumerator1
   ) {module=DynamicMaterialParameters001}
  ]
  rendererClasses [&NiagaraSpriteRendererProperties, &NiagaraMeshRendererProperties]
  renderers [
   {rendererClass=&NiagaraSpriteRendererProperties}
   {rendererClass=&NiagaraSpriteRendererProperties, rendererIndex=1}
   {rendererClass=&NiagaraMeshRendererProperties, rendererIndex=2}
   {rendererClass=&NiagaraSpriteRendererProperties, rendererIndex=3}
  ]
 }
 NE_VC_DriftTrail_Sprites {
  emitterUpdateScript [
   &EmitterState()
   &SpawnRate(SpawnRate=110, Use Spawn Probability=true)
   &SpawnRate(SpawnGroup=1, SpawnRate=160, Use Spawn Probability=true) {module=SpawnRate001}
   &SpawnRate(SpawnGroup=2, SpawnRate=80, Use Spawn Probability=false) {module=SpawnRate002}
  ]
  particleSpawnScript [
   &VC_PaintedVariantsOverride() {enabled=false}
   /SetVariables(Particles.VisibilityTag=$Emitter.SpawnGroup) {isSetParameters=true, module=SetVariables_#5}
   &InitializeParticle(
    Color Mode=::NewEnumerator0
    Lifetime Max=0.65
    Lifetime Min=0.3
    Lifetime Mode=::NewEnumerator1
    Sprite Rotation Angle=&RandomRangeFloat(Minimum=-0.02, Maximum=0.02)
    Sprite Rotation Mode=::NewEnumerator1
    Sprite Size Max=&SelectVector2DFromArray(
     Direct Array Index=&SpawnGroup()
     Vector 2D Selection Array=DataInterface() {propertyValues {FloatData [<56 56>, <50 50>, <15 15>]}}
    )
    Sprite Size Min=&SelectVector2DFromArray(
     Direct Array Index=&SpawnGroup()
     Vector 2D Selection Array=DataInterface() {propertyValues {FloatData [<38 38>, <35 35>, <7.5 7.5>]}}
    )
    Sprite Size Mode=::NewEnumerator2
    UsePositionOffset=false
    Write Material Random=true
   )
   &PartitionParticles()
   &SkeletalMeshLocation(
    Bone / Socket Index=&Add_Integer(A=$Particles.PartitionParticles.Partition, B=4)
    Bone / Socket Sampling Mode=::NewEnumerator5
    MaskSpawnGroup=false
    Mesh Sampling Type=::NewEnumerator6
    Sampled Position Offset=&AddVectorToPosition(Position=<-50 0 -15>, Vector=$User.offset)
    Skeletal Mesh=$System.SkeletalMesh
   )
   &AddVelocity(Velocity=&MergeHelper_VectorSwitch(
     0=&RandomRangeVector(Minimum=<-350 -70 -40>, Maximum=<-500 70 82.5>)
     1=&RandomRangeVector(Minimum=<-400 -70 -40>, Maximum=<-625 70 82.5>)
     2=&RandomRangeVector(Minimum=<-350 -70 -40>, Maximum=<-500 70 82.5>)
     3=&RandomRangeVector(Minimum=<-200 -70 -40>, Maximum=<-350 70 82.5>)
     Index=$System.GameModeID
     Num=::NewEnumerator2
    ))
   &ShapeLocation(
    Hemisphere Distribution=<1 0.5>
    MaskSpawnGroup=false
    Offset=<-36 0 -15>
    Offset Coordinate Space=::Local
    Offset Mode=::NewEnumerator0
    Override Seed=false
    Sphere Radius=&SelectFloatFromArray(Direct Array Index=&SpawnGroup(), Float Selection Array=DataInterface() {propertyValues {FloatData [24,20,24]}})
    Sphere Surface Distribution=0.25
    UseOwnerScale=true
   )
  ]
  particleUpdateScript [
   &VC_PaintedVariantsOverride(
    Black=&MultiplyLinearColor(
     Linear Color=&MakeLinearColorFromVectorAndFloat(
      "Vector (RGB)"=&Multiply_VectorByFloat(Vector=<1 1 1>, Float=&RandomRangeFloat(Minimum=0.16, Maximum=0.2))
      "Float (Alpha)"=&RandomRangeFloat(Minimum=0.7, Maximum=0.925)
     )
     Scale Factor=&MakeVector4FromVectorAndFloat(
      "Vector (RGB)"=&VectorFromFloat(Value=&FloatFromCurve(FloatCurve=DataInterface() {propertyValues {Curve {
           defaultValue=1
           keys [
            {interpMode=RCIM_Cubic, tangentMode=RCTM_User, time=0, value=1}
            {interpMode=RCIM_Cubic, tangentMode=RCTM_User, time=0.25, value=2}
            <at 0.5 1 RCIM_Cubic>
            <at 0.75 2 RCIM_Cubic>
            {tangentMode=RCTM_User, time=1, value=1}
           ]
          }}}))
      "Float (Alpha)"=&FloatFromCurve(FloatCurve=DataInterface() {propertyValues {Curve {
          defaultValue=1
          keys [
           {interpMode=RCIM_Cubic, tangentMode=RCTM_User, time=0, value=0}
           <at 0.3 1>
           {arriveTangent=-1.4286, interpMode=RCIM_Cubic, leaveTangent=-1.4286, time=0.6, value=1}
           {tangentMode=RCTM_User, time=1, value=0}
          ]
         }}})
     )
    )
    BurntSienna=&MultiplyLinearColor(
     Linear Color=&MakeLinearColorFromVectorAndFloat(
      "Vector (RGB)"=&RandomRangeVector(
       Minimum=&MakeVectorFromLinearColorRGB(Color=<rgba 0.4287 0.2016 0.0931 1>)
       Maximum=&MakeVectorFromLinearColorRGB(Color=<rgba 0.42 0.25 0.0792 1>)
      )
      "Float (Alpha)"=&RandomRangeFloat(Minimum=0.2, Maximum=0.8)
     )
     Scale Factor=&MakeVector4FromVectorAndFloat(
      "Vector (RGB)"=&VectorFromFloat(Value=&FloatFromCurve(FloatCurve=DataInterface() {propertyValues {Curve {
           defaultValue=1
           keys [
            {interpMode=RCIM_Cubic, tangentMode=RCTM_User, time=0, value=1}
            {interpMode=RCIM_Cubic, tangentMode=RCTM_User, time=0.25, value=2}
            <at 0.5 1 RCIM_Cubic>
            <at 0.75 2 RCIM_Cubic>
            {tangentMode=RCTM_User, time=1, value=1}
           ]
          }}}))
      "Float (Alpha)"=&FloatFromCurve(FloatCurve=DataInterface() {propertyValues {Curve {
          defaultValue=1
          keys [
           {interpMode=RCIM_Cubic, tangentMode=RCTM_User, time=0, value=0}
           <at 0.3 1>
           {arriveTangent=-1.4286, interpMode=RCIM_Cubic, leaveTangent=-1.4286, time=0.6, value=1}
           {tangentMode=RCTM_User, time=1, value=0}
          ]
         }}})
     )
    )
    Cobalt=&MultiplyLinearColor(
     Linear Color=&MakeLinearColorFromVectorAndFloat(
      "Vector (RGB)"=&RandomRangeVector(
       Minimum=&MakeVectorFromLinearColorRGB(Color=<rgba 0.05 0.1 1 1>)
       Maximum=&MakeVectorFromLinearColorRGB(Color=<rgba 0.275 0.225 1 1>)
      )
      "Float (Alpha)"=&RandomRangeFloat(Minimum=0.2, Maximum=0.85)
     )
     Scale Factor=&MakeVector4FromVectorAndFloat(
      "Vector (RGB)"=&VectorFromFloat(Value=&FloatFromCurve(FloatCurve=DataInterface() {propertyValues {Curve {
           defaultValue=1
           keys [
            {interpMode=RCIM_Cubic, tangentMode=RCTM_User, time=0, value=1}
            {interpMode=RCIM_Cubic, tangentMode=RCTM_User, time=0.25, value=2}
            <at 0.5 1 RCIM_Cubic>
            <at 0.75 2 RCIM_Cubic>
            {tangentMode=RCTM_User, time=1, value=1}
           ]
          }}}))
      "Float (Alpha)"=&FloatFromCurve(FloatCurve=DataInterface() {propertyValues {Curve {
          defaultValue=1
          keys [
           {interpMode=RCIM_Cubic, tangentMode=RCTM_User, time=0, value=0}
           <at 0.3 1>
           {arriveTangent=-1.4286, interpMode=RCIM_Cubic, leaveTangent=-1.4286, time=0.6, value=1}
           {tangentMode=RCTM_User, time=1, value=0}
          ]
         }}})
     )
    )
    Crimson=&MultiplyLinearColor(
     Linear Color=&MakeLinearColorFromVectorAndFloat(
      "Vector (RGB)"=&RandomRangeVector(
       Minimum=&MakeVectorFromLinearColorRGB(Color=<rgba 0.4 0.15 0.15 1>)
       Maximum=&MakeVectorFromLinearColorRGB(Color=<rgba 0.45 0.24 0.24 1>)
      )
      "Float (Alpha)"=&RandomRangeFloat(Minimum=0.2, Maximum=0.85)
     )
     Scale Factor=&MakeVector4FromVectorAndFloat(
      "Vector (RGB)"=&VectorFromFloat(Value=&FloatFromCurve(FloatCurve=DataInterface() {propertyValues {Curve {
           defaultValue=1
           keys [
            {interpMode=RCIM_Cubic, tangentMode=RCTM_User, time=0, value=1}
            {interpMode=RCIM_Cubic, tangentMode=RCTM_User, time=0.25, value=2}
            <at 0.5 1 RCIM_Cubic>
            <at 0.75 2 RCIM_Cubic>
            {tangentMode=RCTM_User, time=1, value=1}
           ]
          }}}))
      "Float (Alpha)"=&FloatFromCurve(FloatCurve=DataInterface() {propertyValues {Curve {
          defaultValue=1
          keys [
           {interpMode=RCIM_Cubic, tangentMode=RCTM_User, time=0, value=0}
           <at 0.3 1>
           {arriveTangent=-1.4286, interpMode=RCIM_Cubic, leaveTangent=-1.4286, time=0.6, value=1}
           {tangentMode=RCTM_User, time=1, value=0}
          ]
         }}})
     )
    )
    DefaultColor=&MultiplyLinearColor(
     Linear Color=&MakeLinearColorFromVectorAndFloat(
      "Vector (RGB)"=&RandomRangeVector(
       Minimum=&MakeVectorFromLinearColorRGB(Color=<rgba 1 0.3325 0.1022 1>)
       Maximum=&MakeVectorFromLinearColorRGB(Color=<rgba 0.8281 0.7858 0.2847 1>)
      )
      "Float (Alpha)"=&RandomRangeFloat(Minimum=0.2, Maximum=0.8)
     )
     Scale Factor=&MakeVector4FromVectorAndFloat(
      "Vector (RGB)"=&VectorFromFloat(Value=&FloatFromCurve(FloatCurve=DataInterface() {propertyValues {Curve {
           defaultValue=1
           keys [
            {interpMode=RCIM_Cubic, tangentMode=RCTM_User, time=0, value=1}
            {interpMode=RCIM_Cubic, tangentMode=RCTM_User, time=0.25, value=2}
            <at 0.5 1 RCIM_Cubic>
            <at 0.75 2 RCIM_Cubic>
            {tangentMode=RCTM_User, time=1, value=1}
           ]
          }}}))
      "Float (Alpha)"=&FloatFromCurve(FloatCurve=DataInterface() {propertyValues {Curve {
          defaultValue=1
          keys [
           {interpMode=RCIM_Cubic, tangentMode=RCTM_User, time=0, value=0}
           <at 0.3 1>
           {arriveTangent=-1.4286, interpMode=RCIM_Cubic, leaveTangent=-1.4286, time=0.6, value=1}
           {tangentMode=RCTM_User, time=1, value=0}
          ]
         }}})
     )
    )
    ForestGreen=&MultiplyLinearColor(
     Linear Color=&MakeLinearColorFromVectorAndFloat(
      "Vector (RGB)"=&RandomRangeVector(
       Minimum=&MakeVectorFromLinearColorRGB(Color=<rgba 0.2 0.3 0.2 1>)
       Maximum=&MakeVectorFromLinearColorRGB(Color=<rgba 0.25 0.5 0.25 1>)
      )
      "Float (Alpha)"=&RandomRangeFloat(Minimum=0.2, Maximum=0.85)
     )
     Scale Factor=&MakeVector4FromVectorAndFloat(
      "Vector (RGB)"=&VectorFromFloat(Value=&FloatFromCurve(FloatCurve=DataInterface() {propertyValues {Curve {
           defaultValue=1
           keys [
            {interpMode=RCIM_Cubic, tangentMode=RCTM_User, time=0, value=1}
            {interpMode=RCIM_Cubic, tangentMode=RCTM_User, time=0.25, value=2}
            <at 0.5 1 RCIM_Cubic>
            <at 0.75 2 RCIM_Cubic>
            {tangentMode=RCTM_User, time=1, value=1}
           ]
          }}}))
      "Float (Alpha)"=&FloatFromCurve(FloatCurve=DataInterface() {propertyValues {Curve {
          defaultValue=1
          keys [
           {interpMode=RCIM_Cubic, tangentMode=RCTM_User, time=0, value=0}
           <at 0.3 1>
           {arriveTangent=-1.4286, interpMode=RCIM_Cubic, leaveTangent=-1.4286, time=0.6, value=1}
           {tangentMode=RCTM_User, time=1, value=0}
          ]
         }}})
     )
    )
    Grey=&MultiplyLinearColor(
     Linear Color=&MakeLinearColorFromVectorAndFloat(
      "Vector (RGB)"=&Multiply_VectorByFloat(Vector=<1 1 1>, Float=&RandomRangeFloat(Minimum=0.26, Maximum=0.3))
      "Float (Alpha)"=&RandomRangeFloat(Minimum=0.3, Maximum=0.8)
     )
     Scale Factor=&MakeVector4FromVectorAndFloat(
      "Vector (RGB)"=&VectorFromFloat(Value=&FloatFromCurve(FloatCurve=DataInterface() {propertyValues {Curve {
           defaultValue=1
           keys [
            {interpMode=RCIM_Cubic, tangentMode=RCTM_User, time=0, value=1}
            {interpMode=RCIM_Cubic, tangentMode=RCTM_User, time=0.25, value=2}
            <at 0.5 1 RCIM_Cubic>
            <at 0.75 2 RCIM_Cubic>
            {tangentMode=RCTM_User, time=1, value=1}
           ]
          }}}))
      "Float (Alpha)"=&FloatFromCurve(FloatCurve=DataInterface() {propertyValues {Curve {
          defaultValue=1
          keys [
           {interpMode=RCIM_Cubic, tangentMode=RCTM_User, time=0, value=0}
           <at 0.3 1>
           {arriveTangent=-1.4286, interpMode=RCIM_Cubic, leaveTangent=-1.4286, time=0.6, value=1}
           {tangentMode=RCTM_User, time=1, value=0}
          ]
         }}})
     )
    )
    Lime=&MultiplyLinearColor(
     Linear Color=&MakeLinearColorFromVectorAndFloat(
      "Vector (RGB)"=&RandomRangeVector(
       Minimum=&MakeVectorFromLinearColorRGB(Color=<rgba 0.1 0.4 0 1>)
       Maximum=&MakeVectorFromLinearColorRGB(Color=<rgba 0.6 0.7 0.05 1>)
      )
      "Float (Alpha)"=&RandomRangeFloat(Minimum=0.2, Maximum=0.85)
     )
     Scale Factor=&MakeVector4FromVectorAndFloat(
      "Vector (RGB)"=&VectorFromFloat(Value=&FloatFromCurve(FloatCurve=DataInterface() {propertyValues {Curve {
           defaultValue=1
           keys [
            {interpMode=RCIM_Cubic, tangentMode=RCTM_User, time=0, value=1}
            {interpMode=RCIM_Cubic, tangentMode=RCTM_User, time=0.25, value=2}
            <at 0.5 1 RCIM_Cubic>
            <at 0.75 2 RCIM_Cubic>
            {tangentMode=RCTM_User, time=1, value=1}
           ]
          }}}))
      "Float (Alpha)"=&FloatFromCurve(FloatCurve=DataInterface() {propertyValues {Curve {
          defaultValue=1
          keys [
           {interpMode=RCIM_Cubic, tangentMode=RCTM_User, time=0, value=0}
           <at 0.3 1>
           {arriveTangent=-1.4286, interpMode=RCIM_Cubic, leaveTangent=-1.4286, time=0.6, value=1}
           {tangentMode=RCTM_User, time=1, value=0}
          ]
         }}})
     )
    )
    Orange=&MultiplyLinearColor(
     Linear Color=&MakeLinearColorFromVectorAndFloat(
      "Vector (RGB)"=&RandomRangeVector(
       Minimum=&MakeVectorFromLinearColorRGB(Color=<rgba 0.7 0.3 0.01 1>)
       Maximum=&MakeVectorFromLinearColorRGB(Color=<rgba 0.8 0.5 0.3 1>)
      )
      "Float (Alpha)"=&RandomRangeFloat(Minimum=0.2, Maximum=0.85)
     )
     Scale Factor=&MakeVector4FromVectorAndFloat(
      "Vector (RGB)"=&VectorFromFloat(Value=&FloatFromCurve(FloatCurve=DataInterface() {propertyValues {Curve {
           defaultValue=1
           keys [
            {interpMode=RCIM_Cubic, tangentMode=RCTM_User, time=0, value=1}
            {interpMode=RCIM_Cubic, tangentMode=RCTM_User, time=0.25, value=2}
            <at 0.5 1 RCIM_Cubic>
            <at 0.75 2 RCIM_Cubic>
            {tangentMode=RCTM_User, time=1, value=1}
           ]
          }}}))
      "Float (Alpha)"=&FloatFromCurve(FloatCurve=DataInterface() {propertyValues {Curve {
          defaultValue=1
          keys [
           {interpMode=RCIM_Cubic, tangentMode=RCTM_User, time=0, value=0}
           <at 0.3 1>
           {arriveTangent=-1.4286, interpMode=RCIM_Cubic, leaveTangent=-1.4286, time=0.6, value=1}
           {tangentMode=RCTM_User, time=1, value=0}
          ]
         }}})
     )
    )
    Pink=&MultiplyLinearColor(
     Linear Color=&MakeLinearColorFromVectorAndFloat(
      "Vector (RGB)"=&RandomRangeVector(
       Minimum=&MakeVectorFromLinearColorRGB(Color=<rgba 0.55 0.145 0.4 1>)
       Maximum=&MakeVectorFromLinearColorRGB(Color=<rgba 0.65 0.35 0.5 1>)
      )
      "Float (Alpha)"=&RandomRangeFloat(Minimum=0.15, Maximum=0.8)
     )
     Scale Factor=&MakeVector4FromVectorAndFloat(
      "Vector (RGB)"=&VectorFromFloat(Value=&FloatFromCurve(FloatCurve=DataInterface() {propertyValues {Curve {
           defaultValue=1
           keys [
            {interpMode=RCIM_Cubic, tangentMode=RCTM_User, time=0, value=1}
            {interpMode=RCIM_Cubic, tangentMode=RCTM_User, time=0.25, value=2}
            <at 0.5 1 RCIM_Cubic>
            <at 0.75 2 RCIM_Cubic>
            {tangentMode=RCTM_User, time=1, value=1}
           ]
          }}}))
      "Float (Alpha)"=&FloatFromCurve(FloatCurve=DataInterface() {propertyValues {Curve {
          defaultValue=1
          keys [
           {interpMode=RCIM_Cubic, tangentMode=RCTM_User, time=0, value=0}
           <at 0.3 1>
           {arriveTangent=-1.4286, interpMode=RCIM_Cubic, leaveTangent=-1.4286, time=0.6, value=1}
           {tangentMode=RCTM_User, time=1, value=0}
          ]
         }}})
     )
    )
    Purple=&MultiplyLinearColor(
     Linear Color=&MakeLinearColorFromVectorAndFloat(
      "Vector (RGB)"=&RandomRangeVector(
       Minimum=&MakeVectorFromLinearColorRGB(Color=<rgba 0.275 0.225 0.325 1>)
       Maximum=&MakeVectorFromLinearColorRGB(Color=<rgba 0.35 0.3 0.45 1>)
      )
      "Float (Alpha)"=&RandomRangeFloat(Minimum=0.2, Maximum=0.875)
     )
     Scale Factor=&MakeVector4FromVectorAndFloat(
      "Vector (RGB)"=&VectorFromFloat(Value=&FloatFromCurve(FloatCurve=DataInterface() {propertyValues {Curve {
           defaultValue=1
           keys [
            {interpMode=RCIM_Cubic, tangentMode=RCTM_User, time=0, value=1}
            {interpMode=RCIM_Cubic, tangentMode=RCTM_User, time=0.25, value=2}
            <at 0.5 1 RCIM_Cubic>
            <at 0.75 2 RCIM_Cubic>
            {tangentMode=RCTM_User, time=1, value=1}
           ]
          }}}))
      "Float (Alpha)"=&FloatFromCurve(FloatCurve=DataInterface() {propertyValues {Curve {
          defaultValue=1
          keys [
           {interpMode=RCIM_Cubic, tangentMode=RCTM_User, time=0, value=0}
           <at 0.3 1>
           {arriveTangent=-1.4286, interpMode=RCIM_Cubic, leaveTangent=-1.4286, time=0.6, value=1}
           {tangentMode=RCTM_User, time=1, value=0}
          ]
         }}})
     )
    )
    Saffron=&MultiplyLinearColor(
     Linear Color=&MakeLinearColorFromVectorAndFloat(
      "Vector (RGB)"=&RandomRangeVector(
       Minimum=&MakeVectorFromLinearColorRGB(Color=<rgba 0.5 0.5 0.1 1>)
       Maximum=&MakeVectorFromLinearColorRGB(Color=<rgba 0.75 0.75 0.25 1>)
      )
      "Float (Alpha)"=&RandomRangeFloat(Minimum=0.15, Maximum=0.8)
     )
     Scale Factor=&MakeVector4FromVectorAndFloat(
      "Vector (RGB)"=&VectorFromFloat(Value=&FloatFromCurve(FloatCurve=DataInterface() {propertyValues {Curve {
           defaultValue=1
           keys [
            {interpMode=RCIM_Cubic, tangentMode=RCTM_User, time=0, value=1}
            {interpMode=RCIM_Cubic, tangentMode=RCTM_User, time=0.25, value=2}
            <at 0.5 1 RCIM_Cubic>
            <at 0.75 2 RCIM_Cubic>
            {tangentMode=RCTM_User, time=1, value=1}
           ]
          }}}))
      "Float (Alpha)"=&FloatFromCurve(FloatCurve=DataInterface() {propertyValues {Curve {
          defaultValue=1
          keys [
           {interpMode=RCIM_Cubic, tangentMode=RCTM_User, time=0, value=0}
           <at 0.3 1>
           {arriveTangent=-1.4286, interpMode=RCIM_Cubic, leaveTangent=-1.4286, time=0.6, value=1}
           {tangentMode=RCTM_User, time=1, value=0}
          ]
         }}})
     )
    )
    SkyBlue=&MultiplyLinearColor(
     Linear Color=&MakeLinearColorFromVectorAndFloat(
      "Vector (RGB)"=&RandomRangeVector(
       Minimum=&MakeVectorFromLinearColorRGB(Color=<rgba 0.07 0.4 0.4 1>)
       Maximum=&MakeVectorFromLinearColorRGB(Color=<rgba 0.35 0.5 0.5 1>)
      )
      "Float (Alpha)"=&RandomRangeFloat(Minimum=0.15, Maximum=0.8)
     )
     Scale Factor=&MakeVector4FromVectorAndFloat(
      "Vector (RGB)"=&VectorFromFloat(Value=&FloatFromCurve(FloatCurve=DataInterface() {propertyValues {Curve {
           defaultValue=1
           keys [
            {interpMode=RCIM_Cubic, tangentMode=RCTM_User, time=0, value=1}
            {interpMode=RCIM_Cubic, tangentMode=RCTM_User, time=0.25, value=2}
            <at 0.5 1 RCIM_Cubic>
            <at 0.75 2 RCIM_Cubic>
            {tangentMode=RCTM_User, time=1, value=1}
           ]
          }}}))
      "Float (Alpha)"=&FloatFromCurve(FloatCurve=DataInterface() {propertyValues {Curve {
          defaultValue=1
          keys [
           {interpMode=RCIM_Cubic, tangentMode=RCTM_User, time=0, value=0}
           <at 0.3 1>
           {arriveTangent=-1.4286, interpMode=RCIM_Cubic, leaveTangent=-1.4286, time=0.6, value=1}
           {tangentMode=RCTM_User, time=1, value=0}
          ]
         }}})
     )
    )
    TitaniumWhite=&MultiplyLinearColor(
     Linear Color=&MakeLinearColorFromVectorAndFloat(
      "Vector (RGB)"=&Multiply_VectorByFloat(Vector=<1 1 1>, Float=&RandomRangeFloat(Minimum=0.45, Maximum=0.8))
      "Float (Alpha)"=&RandomRangeFloat(Minimum=0.1, Maximum=0.7)
     )
     Scale Factor=&MakeVector4FromVectorAndFloat(
      "Vector (RGB)"=&VectorFromFloat(Value=&FloatFromCurve(FloatCurve=DataInterface() {propertyValues {Curve {
           defaultValue=1
           keys [
            {interpMode=RCIM_Cubic, tangentMode=RCTM_User, time=0, value=1}
            {interpMode=RCIM_Cubic, tangentMode=RCTM_User, time=0.25, value=2}
            <at 0.5 1 RCIM_Cubic>
            <at 0.75 2 RCIM_Cubic>
            {tangentMode=RCTM_User, time=1, value=1}
           ]
          }}}))
      "Float (Alpha)"=&FloatFromCurve(FloatCurve=DataInterface() {propertyValues {Curve {
          defaultValue=1
          keys [
           {interpMode=RCIM_Cubic, tangentMode=RCTM_User, time=0, value=0}
           <at 0.3 1>
           {arriveTangent=-1.4286, interpMode=RCIM_Cubic, leaveTangent=-1.4286, time=0.6, value=1}
           {tangentMode=RCTM_User, time=1, value=0}
          ]
         }}})
     )
    )
   ) {module=VC_PaintedVariantsOverride001}
   &FN_DisableTireFX(
    Mask Index Attribute:NiagaraInt32=$Particles.PartitionParticles.Partition
    Mask Index FL=2
    Mask Index FR=3
    Mask Index RL=0
    Mask Index RR=1
   )
   &ParticleState()
   &SpriteRotationRate(Rotation Rate=&MaskFloatBySpawnGroup(
     Masked Float=&RandomRangeFloat(Minimum=-50, Maximum=50)
     Passthrough Float=&MaskFloatBySpawnGroup(
      Masked Float=&RandomRangeFloat(Minimum=-500, Maximum=500)
      Passthrough Float=&RandomRangeFloat(Minimum=-50, Maximum=50)
      Spawn Group Mask=1
     )
    ))
   &InheritSourceMovement(
    Applied Position Falloff Distance=20
    Applied Position Scale=&MergeHelper_VectorSwitch(
     0=<1 1 1>
     1=&VectorFromFloat(Value=&FloatFromCurve(FloatCurve=DataInterface() {propertyValues {Curve {
          defaultValue=0
          keys [{interpMode=RCIM_Cubic, tangentMode=RCTM_User, time=0, value=1}, {interpMode=RCIM_Cubic, tangentMode=RCTM_User, time=0.5, value=0.7}]
         }}}))
     2=<1 1 1>
     3=<1 1 1>
     Index=$System.GameModeID
     Num=::NewEnumerator2
    )
    ApplyForce=false
    ApplyPosition=true
    ApplyPositionFalloff=false
   )
   &ScaleSpriteSize(
    Scale Sprite Size Mode=::NewEnumerator0
    Uniform Scale Factor=&Multiply_Float(
     A=1.1
     B=&FloatFromCurve(FloatCurve=DataInterface() {propertyValues {Curve {defaultValue=1, keys [<at 0 0.2 RCIM_Cubic>,<at 0.35 1 RCIM_Cubic>]}}})
    )
   )
   &SubUVAnimation(
    Evaluation Type=::NewEnumerator0
    "Sprite Renderer "=DataInterface() {propertyValues {SpriteRenderer=&self:NE_VC_DriftTrail_Sprites_0.NiagaraSpriteRendererProperties_5}}
    SubUV Animation Mode:ENiagara_SubUVLookupModeV2=::NewEnumerator2
    UseEndFrame=false
    UseStartFrame=false
   )
   &SolveForcesAndVelocity(Clamp Velocity=false, Limit Acceleration=false)
   /SetVariables(Particles.IntensifiedColor:LinearColor=&Multiply_LinearColorByFloat(Linear Color=$Particles.Color, Float=4.5)) {isSetParameters=true}
  ]
  rendererClasses [&NiagaraSpriteRendererProperties]
  renderers [
   {rendererClass=&NiagaraSpriteRendererProperties}
   {rendererClass=&NiagaraSpriteRendererProperties, rendererIndex=1}
   {rendererClass=&NiagaraSpriteRendererProperties, rendererIndex=2}
  ]
 }
 NE_VC_Trails {
  emitterUpdateScript [
   &EmitterState(Inactive Response=::NewEnumerator1, Life Cycle Mode=::NewEnumerator1, Loop Duration=1, UseLoopDelay=false)
   &VC_RateQualityHandler(Epic Spawn Rate=30, High Spawn Rate=30, Low Spawn Rate=30, Use Custom Rate?=true)
   &VC_SpawnRibbon_v1(Number of Ribbons=2, SpawnRate=$Emitter.Spawn Rate)
   /SetVariables(Emitter.OffsetDissolve:NiagaraFloat=&Add_Float(A=0.05, B=&Sine(Normalized Angle=$Emitter.Age, Period=0.5, Scale=0.1, Bias=0))) {enabled=false, isSetParameters=true}
  ]
  particleSpawnScript [
   &VC_PaintedVariantsOverride(Use R1 Painted Overrides=true) {enabled=false}
   &VC_InitializeRibbon_v1(
    Custom Offset=true
    Enable Skeletal Position=true
    Facing Axis:ENiagaraOrientationAxis=::ZAxis
    Lifetime=&MergeHelper_FloatSwitch(0=1.7, 1=1.1, 2=1.7, 3=1.7, Index=$System.GameModeID, Num=::NewEnumerator2)
    Local Vector Offset=&AddVectorToPosition(Position=<40 0 -50>, Vector=Linked() {linkedVariable {name=User.DTCarOffset, type=NiagaraPosition}})
    Offset Amount=1
    Ribbon Width=0
   )
   &AddVelocity(Velocity=&MergeHelper_VectorSwitch(0=<-1050 0 0>, 1=<-1000 0 0>, 2=<-1050 0 0>, 3=<-850 0 0>, Index=$System.GameModeID, Num=::NewEnumerator2)) {module=AddVelocity001}
   /SetVariables(Particles.Separator:NiagaraInt32=&SpawnGroup()) {isSetParameters=true}
   /VehicleCosmetics/FX/VC_FXParents/Modules/DLR_ProductPreviews(GameModeID:NiagaraInt32=$System.GameModeID, bPreviewVehicleSpeed:NiagaraBool=$System.bPreviewVehicleSpeed)
  ]
  particleUpdateScript [
   &VC_PaintedVariantsOverride(
    Black=&MultiplyLinearColor(
     Linear Color=<rgba 0 0 0 1>
     Scale Factor=&MakeVector4FromVectorAndFloat(
      "Vector (RGB)"=<1 1 1>
      "Float (Alpha)"=&FloatFromCurve(
       FloatCurve=DataInterface() {propertyValues {Curve {defaultValue=0, keys [{tangentMode=RCTM_User, time=0, value=1}, {tangentMode=RCTM_User, time=0.85, value=0}]}}}
       Scale Curve=2.8
      )
     )
    )
    BurntSienna=&MultiplyLinearColor(
     Linear Color=<rgba 0.1512 0.0714 0.0285 2.25>
     Scale Factor=&MakeVector4FromVectorAndFloat(
      "Vector (RGB)"=&VectorFromFloat(Value=1)
      "Float (Alpha)"=&FloatFromCurve(
       FloatCurve=DataInterface() {propertyValues {Curve {defaultValue=0, keys [{tangentMode=RCTM_User, time=0, value=1}, {tangentMode=RCTM_User, time=0.85, value=0}]}}}
       Scale Curve=2
      )
     )
    )
    Cobalt=&MultiplyLinearColor(
     Linear Color=<rgba 0.05 0.1 1 1>
     Scale Factor=&MakeVector4FromVectorAndFloat(
      "Vector (RGB)"=<0.2 0.2 0.2>
      "Float (Alpha)"=&FloatFromCurve(
       FloatCurve=DataInterface() {propertyValues {Curve {defaultValue=0, keys [{tangentMode=RCTM_User, time=0, value=1}, {tangentMode=RCTM_User, time=0.85, value=0}]}}}
       Scale Curve=2.6
      )
     )
    )
    Crimson=&MultiplyLinearColor(
     Linear Color=<rgba 0.5 0 0 1>
     Scale Factor=&MakeVector4FromVectorAndFloat(
      "Vector (RGB)"=<0.4 0.4 0.4>
      "Float (Alpha)"=&FloatFromCurve(
       FloatCurve=DataInterface() {propertyValues {Curve {defaultValue=0, keys [{tangentMode=RCTM_User, time=0, value=1}, {tangentMode=RCTM_User, time=0.85, value=0}]}}}
       Scale Curve=3
      )
     )
    )
    DefaultColor=&MultiplyLinearColor(
     Linear Color=<rgba 0 0 0 1>
     Scale Factor=&MakeVector4FromVectorAndFloat(
      "Vector (RGB)"=<1 1 1>
      "Float (Alpha)"=&FloatFromCurve(
       FloatCurve=DataInterface() {propertyValues {Curve {defaultValue=0, keys [{tangentMode=RCTM_User, time=0, value=1}, {tangentMode=RCTM_User, time=0.85, value=0}]}}}
       Scale Curve=2.7
      )
     )
    )
    ForestGreen=&MultiplyLinearColor(
     Linear Color=<rgba 0 0.35 0 1>
     Scale Factor=&MakeVector4FromVectorAndFloat(
      "Vector (RGB)"=<0.4 0.4 0.4>
      "Float (Alpha)"=&FloatFromCurve(
       FloatCurve=DataInterface() {propertyValues {Curve {defaultValue=0, keys [{tangentMode=RCTM_User, time=0, value=1}, {tangentMode=RCTM_User, time=0.85, value=0}]}}}
       Scale Curve=3
      )
     )
    )
    Grey=&MultiplyLinearColor(
     Linear Color=<rgba 0.125 0.125 0.125 1>
     Scale Factor=&MakeVector4FromVectorAndFloat(
      "Vector (RGB)"=<1.25 1.25 1.25>
      "Float (Alpha)"=&FloatFromCurve(
       FloatCurve=DataInterface() {propertyValues {Curve {defaultValue=0, keys [{tangentMode=RCTM_User, time=0, value=1}, {tangentMode=RCTM_User, time=0.85, value=0}]}}}
       Scale Curve=3
      )
     )
    )
    Lime=&MultiplyLinearColor(
     Linear Color=<rgba 0.3 0.7 0 1>
     Scale Factor=&MakeVector4FromVectorAndFloat(
      "Vector (RGB)"=<0.35 0.35 0.35>
      "Float (Alpha)"=&FloatFromCurve(
       FloatCurve=DataInterface() {propertyValues {Curve {defaultValue=0, keys [{tangentMode=RCTM_User, time=0, value=1}, {tangentMode=RCTM_User, time=0.85, value=0}]}}}
       Scale Curve=3
      )
     )
    )
    Orange=&MultiplyLinearColor(
     Linear Color=<rgba 0.7 0.35 0.01 1>
     Scale Factor=&MakeVector4FromVectorAndFloat(
      "Vector (RGB)"=<0.4 0.4 0.4>
      "Float (Alpha)"=&FloatFromCurve(
       FloatCurve=DataInterface() {propertyValues {Curve {defaultValue=0, keys [{tangentMode=RCTM_User, time=0, value=1}, {tangentMode=RCTM_User, time=0.85, value=0}]}}}
       Scale Curve=3
      )
     )
    )
    Pink=&MultiplyLinearColor(
     Linear Color=<rgba 0.55 0.145 0.4 1>
     Scale Factor=&MakeVector4FromVectorAndFloat(
      "Vector (RGB)"=<0.4 0.4 0.4>
      "Float (Alpha)"=&FloatFromCurve(
       FloatCurve=DataInterface() {propertyValues {Curve {defaultValue=0, keys [{tangentMode=RCTM_User, time=0, value=1}, {tangentMode=RCTM_User, time=0.85, value=0}]}}}
       Scale Curve=3
      )
     )
    )
    Purple=&MultiplyLinearColor(
     Linear Color=<rgba 0.25 0 0.3 1>
     Scale Factor=&MakeVector4FromVectorAndFloat(
      "Vector (RGB)"=<0.5 0.5 0.5>
      "Float (Alpha)"=&FloatFromCurve(
       FloatCurve=DataInterface() {propertyValues {Curve {defaultValue=0, keys [{tangentMode=RCTM_User, time=0, value=1}, {tangentMode=RCTM_User, time=0.85, value=0}]}}}
       Scale Curve=3
      )
     )
    )
    R1 Black:LinearColor=&MultiplyLinearColor(
     Linear Color=<rgba 0 0 0 6>
     Scale Factor=&MakeVector4FromVectorAndFloat(
      "Vector (RGB)"=<1 1 1>
      "Float (Alpha)"=&FloatFromCurve(FloatCurve=DataInterface() {propertyValues {Curve {defaultValue=0, keys [{tangentMode=RCTM_User, time=0, value=1}, {tangentMode=RCTM_User, time=1.5, value=0}]}}})
     )
    )
    R1 BurntSienna:LinearColor=&MultiplyLinearColor(
     Linear Color=<rgba 0.1512 0.0714 0.0285 2.25>
     Scale Factor=&MakeVector4FromVectorAndFloat(
      "Vector (RGB)"=&VectorFromFloat(Value=0.1)
      "Float (Alpha)"=&FloatFromCurve(
       FloatCurve=DataInterface() {propertyValues {Curve {defaultValue=0, keys [{tangentMode=RCTM_User, time=0, value=1}, {tangentMode=RCTM_User, time=1.5, value=0}]}}}
       Scale Curve=3
      )
     )
    )
    R1 Cobalt:LinearColor=&MultiplyLinearColor(
     Linear Color=<rgba 0.05 0.1 1 1>
     Scale Factor=&MakeVector4FromVectorAndFloat(
      "Vector (RGB)"=<0.2 0.2 0.2>
      "Float (Alpha)"=&FloatFromCurve(
       FloatCurve=DataInterface() {propertyValues {Curve {defaultValue=0, keys [{tangentMode=RCTM_User, time=0, value=1}, {tangentMode=RCTM_User, time=1.5, value=0}]}}}
       Scale Curve=3
      )
     )
    )
    R1 Crimson:LinearColor=&MultiplyLinearColor(
     Linear Color=<rgba 0.5 0 0 1>
     Scale Factor=&MakeVector4FromVectorAndFloat(
      "Vector (RGB)"=<0.1 0.1 0.1>
      "Float (Alpha)"=&FloatFromCurve(
       FloatCurve=DataInterface() {propertyValues {Curve {defaultValue=0, keys [{tangentMode=RCTM_User, time=0, value=1}, {tangentMode=RCTM_User, time=1.5, value=0}]}}}
       Scale Curve=3
      )
     )
    )
    R1 DefaultColor=&MultiplyLinearColor(
     Linear Color=<rgba 0 0 0 6>
     Scale Factor=&MakeVector4FromVectorAndFloat(
      "Vector (RGB)"=<1 1 1>
      "Float (Alpha)"=&FloatFromCurve(FloatCurve=DataInterface() {propertyValues {Curve {defaultValue=0, keys [{tangentMode=RCTM_User, time=0, value=1}, {tangentMode=RCTM_User, time=1.5, value=0}]}}})
     )
    )
    R1 ForestGreen:LinearColor=&MultiplyLinearColor(
     Linear Color=<rgba 0 0.35 0 1>
     Scale Factor=&MakeVector4FromVectorAndFloat(
      "Vector (RGB)"=<0.1 0.1 0.1>
      "Float (Alpha)"=&FloatFromCurve(
       FloatCurve=DataInterface() {propertyValues {Curve {defaultValue=0, keys [{tangentMode=RCTM_User, time=0, value=1}, {tangentMode=RCTM_User, time=1.5, value=0}]}}}
       Scale Curve=3
      )
     )
    )
    R1 Grey:LinearColor=&MultiplyLinearColor(
     Linear Color=<rgba 0.125 0.125 0.125 1>
     Scale Factor=&MakeVector4FromVectorAndFloat(
      "Vector (RGB)"=<0.1 0.1 0.1>
      "Float (Alpha)"=&FloatFromCurve(
       FloatCurve=DataInterface() {propertyValues {Curve {defaultValue=0, keys [{tangentMode=RCTM_User, time=0, value=1}, {tangentMode=RCTM_User, time=1.5, value=0}]}}}
       Scale Curve=3
      )
     )
    )
    R1 Lime:LinearColor=&MultiplyLinearColor(
     Linear Color=<rgba 0.2143 0.5 0 1>
     Scale Factor=&MakeVector4FromVectorAndFloat(
      "Vector (RGB)"=<0.1 0.1 0.1>
      "Float (Alpha)"=&FloatFromCurve(
       FloatCurve=DataInterface() {propertyValues {Curve {defaultValue=0, keys [{tangentMode=RCTM_User, time=0, value=1}, {tangentMode=RCTM_User, time=1.5, value=0}]}}}
       Scale Curve=3.5
      )
     )
    )
    R1 Orange:LinearColor=&MultiplyLinearColor(
     Linear Color=<rgba 0.6 0.3 0.0086 1>
     Scale Factor=&MakeVector4FromVectorAndFloat(
      "Vector (RGB)"=<0.1 0.1 0.1>
      "Float (Alpha)"=&FloatFromCurve(
       FloatCurve=DataInterface() {propertyValues {Curve {defaultValue=0, keys [{tangentMode=RCTM_User, time=0, value=1}, {tangentMode=RCTM_User, time=1.5, value=0}]}}}
       Scale Curve=3
      )
     )
    )
    R1 Pink:LinearColor=&MultiplyLinearColor(
     Linear Color=<rgba 0.5 0.1318 0.3636 1>
     Scale Factor=&MakeVector4FromVectorAndFloat(
      "Vector (RGB)"=<0.1 0.1 0.1>
      "Float (Alpha)"=&FloatFromCurve(
       FloatCurve=DataInterface() {propertyValues {Curve {defaultValue=0, keys [{tangentMode=RCTM_User, time=0, value=1}, {tangentMode=RCTM_User, time=1.5, value=0}]}}}
       Scale Curve=3
      )
     )
    )
    R1 Purple:LinearColor=&MultiplyLinearColor(
     Linear Color=<rgba 0.25 0 0.3 1>
     Scale Factor=&MakeVector4FromVectorAndFloat(
      "Vector (RGB)"=<0.1 0.1 0.1>
      "Float (Alpha)"=&FloatFromCurve(
       FloatCurve=DataInterface() {propertyValues {Curve {defaultValue=0, keys [{tangentMode=RCTM_User, time=0, value=1}, {tangentMode=RCTM_User, time=1.5, value=0}]}}}
       Scale Curve=3
      )
     )
    )
    R1 Saffron:LinearColor=&MultiplyLinearColor(
     Linear Color=<rgba 0.5 0.4 0 1>
     Scale Factor=&MakeVector4FromVectorAndFloat(
      "Vector (RGB)"=<0.05 0.05 0.05>
      "Float (Alpha)"=&FloatFromCurve(
       FloatCurve=DataInterface() {propertyValues {Curve {defaultValue=0, keys [{tangentMode=RCTM_User, time=0, value=1}, {tangentMode=RCTM_User, time=1.5, value=0}]}}}
       Scale Curve=3.5
      )
     )
    )
    R1 SkyBlue:LinearColor=&MultiplyLinearColor(
     Linear Color=<rgba 0 0.5 0.8 1>
     Scale Factor=&MakeVector4FromVectorAndFloat(
      "Vector (RGB)"=<0.05 0.05 0.05>
      "Float (Alpha)"=&FloatFromCurve(
       FloatCurve=DataInterface() {propertyValues {Curve {defaultValue=0, keys [{tangentMode=RCTM_User, time=0, value=1}, {tangentMode=RCTM_User, time=1.5, value=0}]}}}
       Scale Curve=2.75
      )
     )
    )
    R1 TitaniumWhite:LinearColor=&MultiplyLinearColor(
     Linear Color=<rgba 1 1 1 1>
     Scale Factor=&MakeVector4FromVectorAndFloat(
      "Vector (RGB)"=<0.1 0.1 0.1>
      "Float (Alpha)"=&FloatFromCurve(
       FloatCurve=DataInterface() {propertyValues {Curve {defaultValue=0, keys [{tangentMode=RCTM_User, time=0, value=1}, {tangentMode=RCTM_User, time=1.5, value=0}]}}}
       Scale Curve=3.5
      )
     )
    )
    R2 Black:LinearColor=&MultiplyLinearColor(
     Linear Color=<rgba 0.17 0.17 0.17 1.16>
     Scale Factor=&MakeVector4FromVectorAndFloat(
      "Vector (RGB)"=<1 1 1>
      "Float (Alpha)"=&FloatFromCurve(
       FloatCurve=DataInterface() {propertyValues {Curve {
          defaultValue=1
          keys [
           {interpMode=RCIM_Cubic, tangentMode=RCTM_User, time=0.065, value=0}
           <at 0.175 1>
           {arriveTangent=-3.0769, interpMode=RCIM_Cubic, leaveTangent=-3.0769, time=0.3, value=1}
           <at 0.5 0 RCIM_Cubic>
          ]
         }}}
       Scale Curve=2
      )
     )
    )
    R2 BurntSienna:LinearColor=&MultiplyLinearColor(
     Linear Color=<rgba 0.363 0.1944 0.1 0.93>
     Scale Factor=&MakeVector4FromVectorAndFloat(
      "Vector (RGB)"=<1.2 1.45 1>
      "Float (Alpha)"=&FloatFromCurve(
       FloatCurve=DataInterface() {propertyValues {Curve {
          defaultValue=1
          keys [
           {interpMode=RCIM_Cubic, tangentMode=RCTM_User, time=0.065, value=0}
           <at 0.175 1>
           {arriveTangent=-3.0769, interpMode=RCIM_Cubic, leaveTangent=-3.0769, time=0.3, value=1}
           <at 0.5 0 RCIM_Cubic>
          ]
         }}}
       Scale Curve=1.25
      )
     )
    )
    R2 Cobalt:LinearColor=&MultiplyLinearColor(
     Linear Color=<rgba 0.125 0.175 1 0.9294>
     Scale Factor=&MakeVector4FromVectorAndFloat(
      "Vector (RGB)"=<1.4 1.4 1.4>
      "Float (Alpha)"=&FloatFromCurve(
       FloatCurve=DataInterface() {propertyValues {Curve {
          defaultValue=1
          keys [
           {interpMode=RCIM_Cubic, tangentMode=RCTM_User, time=0.065, value=0}
           <at 0.175 1>
           {arriveTangent=-2.6667, interpMode=RCIM_Cubic, leaveTangent=-2.6667, time=0.25, value=1}
           <at 0.55 0 RCIM_Cubic>
          ]
         }}}
       Scale Curve=0.6
      )
     )
    )
    R2 Crimson:LinearColor=&MultiplyLinearColor(
     Linear Color=<rgba 0.4 0.08 0.11 0.93>
     Scale Factor=&MakeVector4FromVectorAndFloat(
      "Vector (RGB)"=<1.25 2.6 0>
      "Float (Alpha)"=&FloatFromCurve(
       FloatCurve=DataInterface() {propertyValues {Curve {
          defaultValue=1
          keys [
           {interpMode=RCIM_Cubic, tangentMode=RCTM_User, time=0.065, value=0}
           <at 0.155 1>
           {arriveTangent=-2.7397, interpMode=RCIM_Cubic, leaveTangent=-2.7397, time=0.275, value=1}
           <at 0.52 0 RCIM_Cubic>
          ]
         }}}
       Scale Curve=2.1
      )
     )
    )
    R2 DefaultColor:LinearColor=&MultiplyLinearColor(
     Linear Color=<rgba 0.4233 0.402 0.3813 0.9294>
     Scale Factor=&MakeVector4FromVectorAndFloat(
      "Vector (RGB)"=<1 1 1>
      "Float (Alpha)"=&FloatFromCurve(FloatCurve=DataInterface() {propertyValues {Curve {
          defaultValue=1
          keys [
           {interpMode=RCIM_Cubic, tangentMode=RCTM_User, time=0.065, value=0}
           <at 0.175 1>
           {arriveTangent=-2.6667, interpMode=RCIM_Cubic, leaveTangent=-2.6667, time=0.3, value=1}
           <at 0.55 0 RCIM_Cubic>
          ]
         }}})
     )
    )
    R2 ForestGreen:LinearColor=&MultiplyLinearColor(
     Linear Color=<rgba 0 0.24 0 0.93>
     Scale Factor=&MakeVector4FromVectorAndFloat(
      "Vector (RGB)"=<2.5 2.5 2.5>
      "Float (Alpha)"=&FloatFromCurve(
       FloatCurve=DataInterface() {propertyValues {Curve {
          defaultValue=1
          keys [
           {interpMode=RCIM_Cubic, tangentMode=RCTM_User, time=0.065, value=0}
           <at 0.17 1>
           {arriveTangent=-2.8169, interpMode=RCIM_Cubic, leaveTangent=-2.8169, time=0.3, value=1}
           <at 0.525 0 RCIM_Cubic>
          ]
         }}}
       Scale Curve=1.6
      )
     )
    )
    R2 Grey:LinearColor=&MultiplyLinearColor(
     Linear Color=<rgba 0.125 0.125 0.13 0.93>
     Scale Factor=&MakeVector4FromVectorAndFloat(
      "Vector (RGB)"=<3.25 3.25 3.25>
      "Float (Alpha)"=&FloatFromCurve(
       FloatCurve=DataInterface() {propertyValues {Curve {
          defaultValue=1
          keys [
           {interpMode=RCIM_Cubic, tangentMode=RCTM_User, time=0.065, value=0}
           <at 0.175 1>
           {arriveTangent=-2.8571, interpMode=RCIM_Cubic, leaveTangent=-2.8571, time=0.275, value=1}
           <at 0.525 0 RCIM_Cubic>
          ]
         }}}
       Scale Curve=1.55
      )
     )
    )
    R2 Lime:LinearColor=&MultiplyLinearColor(
     Linear Color=<rgba 0.45 0.7 0 0.93>
     Scale Factor=&MakeVector4FromVectorAndFloat(
      "Vector (RGB)"=<1.65 1.65 1>
      "Float (Alpha)"=&FloatFromCurve(
       FloatCurve=DataInterface() {propertyValues {Curve {
          defaultValue=1
          keys [
           {interpMode=RCIM_Cubic, tangentMode=RCTM_User, time=0.065, value=0}
           <at 0.175 1>
           {arriveTangent=-2.8571, interpMode=RCIM_Cubic, leaveTangent=-2.8571, time=0.3, value=1}
           <at 0.525 0 RCIM_Cubic>
          ]
         }}}
       Scale Curve=0.8
      )
     )
    )
    R2 Orange:LinearColor=&MultiplyLinearColor(
     Linear Color=<rgba 0.575 0.2875 0.0082 1>
     Scale Factor=&MakeVector4FromVectorAndFloat(
      "Vector (RGB)"=<1.7 1.9 1.7>
      "Float (Alpha)"=&FloatFromCurve(
       FloatCurve=DataInterface() {propertyValues {Curve {
          defaultValue=1
          keys [
           {interpMode=RCIM_Cubic, tangentMode=RCTM_User, time=0.065, value=0}
           <at 0.175 1>
           {arriveTangent=-2.6667, interpMode=RCIM_Cubic, leaveTangent=-2.6667, time=0.3, value=1}
           <at 0.55 0 RCIM_Cubic>
          ]
         }}}
       Scale Curve=1.4
      )
     )
    )
    R2 Pink:LinearColor=&MultiplyLinearColor(
     Linear Color=<rgba 0.55 0.155 0.45 1>
     Scale Factor=&MakeVector4FromVectorAndFloat(
      "Vector (RGB)"=<2 4 2.5>
      "Float (Alpha)"=&FloatFromCurve(
       FloatCurve=DataInterface() {propertyValues {Curve {
          defaultValue=1
          keys [
           {interpMode=RCIM_Cubic, tangentMode=RCTM_User, time=0.065, value=0}
           <at 0.175 1>
           {arriveTangent=-2.6667, interpMode=RCIM_Cubic, leaveTangent=-2.6667, time=0.3, value=1}
           <at 0.55 0 RCIM_Cubic>
          ]
         }}}
       Scale Curve=0.75
      )
     )
    )
    R2 Purple:LinearColor=&MultiplyLinearColor(
     Linear Color=<rgba 0.15 0.1 0.3 1>
     Scale Factor=&MakeVector4FromVectorAndFloat(
      "Vector (RGB)"=<2.5 2.5 2.5>
      "Float (Alpha)"=&FloatFromCurve(
       FloatCurve=DataInterface() {propertyValues {Curve {
          defaultValue=1
          keys [
           {interpMode=RCIM_Cubic, tangentMode=RCTM_User, time=0.065, value=0}
           <at 0.16 1>
           {arriveTangent=-2.439, interpMode=RCIM_Cubic, leaveTangent=-2.439, time=0.3, value=1}
           <at 0.57 0 RCIM_Cubic>
          ]
         }}}
       Scale Curve=1.8
      )
     )
    )
    R2 Saffron:LinearColor=&MultiplyLinearColor(
     Linear Color=<rgba 0.4 0.37 0 0.93>
     Scale Factor=&MakeVector4FromVectorAndFloat(
      "Vector (RGB)"=<2 2 1>
      "Float (Alpha)"=&FloatFromCurve(
       FloatCurve=DataInterface() {propertyValues {Curve {
          defaultValue=1
          keys [
           {interpMode=RCIM_Cubic, tangentMode=RCTM_User, time=0.065, value=0}
           <at 0.16 1>
           {arriveTangent=-2.7397, interpMode=RCIM_Cubic, leaveTangent=-2.7397, time=0.3, value=1}
           <at 0.525 0 RCIM_Cubic>
          ]
         }}}
       Scale Curve=2
      )
     )
    )
    R2 SkyBlue:LinearColor=&MultiplyLinearColor(
     Linear Color=<rgba 0.15 0.5 0.8 0.9294>
     Scale Factor=&MakeVector4FromVectorAndFloat(
      "Vector (RGB)"=<1.4 1.4 1.4>
      "Float (Alpha)"=&FloatFromCurve(
       FloatCurve=DataInterface() {propertyValues {Curve {
          defaultValue=1
          keys [
           {interpMode=RCIM_Cubic, tangentMode=RCTM_User, time=0.065, value=0}
           <at 0.175 1>
           {arriveTangent=-2.6667, interpMode=RCIM_Cubic, leaveTangent=-2.6667, time=0.25, value=1}
           <at 0.55 0 RCIM_Cubic>
          ]
         }}}
       Scale Curve=1.7
      )
     )
    )
    R2 TitaniumWhite:LinearColor=&MultiplyLinearColor(
     Linear Color=<rgba 1 1 1 0.93>
     Scale Factor=&MakeVector4FromVectorAndFloat(
      "Vector (RGB)"=<1.4 1.4 1.45>
      "Float (Alpha)"=&FloatFromCurve(
       FloatCurve=DataInterface() {propertyValues {Curve {
          defaultValue=1
          keys [
           {interpMode=RCIM_Cubic, tangentMode=RCTM_User, time=0.065, value=0}
           <at 0.16 1>
           {arriveTangent=-2.7397, interpMode=RCIM_Cubic, leaveTangent=-2.7397, time=0.27, value=1}
           <at 0.525 0 RCIM_Cubic>
          ]
         }}}
       Scale Curve=0.54
      )
     )
    )
    Saffron=&MultiplyLinearColor(
     Linear Color=<rgba 0.5 0.4 0 1>
     Scale Factor=&MakeVector4FromVectorAndFloat(
      "Vector (RGB)"=<0.55 0.55 0.55>
      "Float (Alpha)"=&FloatFromCurve(
       FloatCurve=DataInterface() {propertyValues {Curve {defaultValue=0, keys [{tangentMode=RCTM_User, time=0, value=1}, {tangentMode=RCTM_User, time=0.85, value=0}]}}}
       Scale Curve=3
      )
     )
    )
    SkyBlue=&MultiplyLinearColor(
     Linear Color=<rgba 0 0.5 0.8 1>
     Scale Factor=&MakeVector4FromVectorAndFloat(
      "Vector (RGB)"=<0.3 0.3 0.3>
      "Float (Alpha)"=&FloatFromCurve(
       FloatCurve=DataInterface() {propertyValues {Curve {defaultValue=0, keys [{tangentMode=RCTM_User, time=0, value=1}, {tangentMode=RCTM_User, time=0.85, value=0}]}}}
       Scale Curve=2.75
      )
     )
    )
    TitaniumWhite=&MultiplyLinearColor(
     Linear Color=<rgba 1 1 1 1>
     Scale Factor=&MakeVector4FromVectorAndFloat(
      "Vector (RGB)"=<0.4 0.4 0.45>
      "Float (Alpha)"=&FloatFromCurve(
       FloatCurve=DataInterface() {propertyValues {Curve {defaultValue=0, keys [{tangentMode=RCTM_User, time=0, value=1}, {tangentMode=RCTM_User, time=0.85, value=0}]}}}
       Scale Curve=3
      )
     )
    )
    Use R1 Painted Overrides=true
    Use R2 Painted Overrides=true
   ) {module=VC_PaintedVariantsOverride001}
   &FN_DisableTireFX(Mask Index Attribute:NiagaraInt32=$Particles.SpawnGroup, Mask Index FL=2, Mask Index FR=3, Mask Index RL=0, Mask Index RR=1)
   &ParticleState()
   &InheritSourceMovement(
    Applied Position Scale=&VectorFromFloat(Value=&FloatFromCurve(FloatCurve=DataInterface() {propertyValues {Curve {keys [
          {interpMode=RCIM_Cubic, leaveTangent=0.0258, tangentMode=RCTM_Break, time=0, value=1}
          {arriveTangent=-2.9025, interpMode=RCIM_Cubic, leaveTangent=0.0258, tangentMode=RCTM_Break, time=1, value=0}
         ]}}}))
    ApplyForce=false
    ApplyPosition=true
    ApplyPositionFalloff=false
   ) {enabled=false, module=InheritSourceMovement001}
   &DynamicMaterialParameters(
    Index 2 Param 1=&MaskFloat(Mask Source=0, Mask Parameter=?, Mask=0, Masked Float=0.25, Passthrough Float=1)
    Index 2 Param 3=&MaskFloat(Mask Source=0, Mask Parameter=?, Mask=0, Masked Float=0.25, Passthrough Float=1)
    Module.Index2Float0Write=true
    Module.Index2Float1Write=true
    Module.Index2Float2Write=true
    Module.Index2Float3Write=true
    Write Parameter Index 0=::NewEnumerator0
    Write Parameter Index 2=::NewEnumerator1
   )
   &InheritSourceMovement(
    Applied Position Scale=&Multiply_VectorByFloat(
     Vector=<1 1 1>
     Float=&FloatFromCurve(FloatCurve=DataInterface() {propertyValues {Curve {keys [
          {interpMode=RCIM_Cubic, leaveTangent=0.0258, tangentMode=RCTM_Break, time=0, value=1}
          {arriveTangent=-2.9025, interpMode=RCIM_Cubic, leaveTangent=0.0258, tangentMode=RCTM_Break, time=1, value=0}
         ]}}})
    )
    ApplyForce=false
    ApplyPosition=true
    ApplyPositionFalloff=false
   )
   &SolveForcesAndVelocity(Clamp Velocity=false, Limit Acceleration=false)
   /SetVariables(
    Particles.DustPosition:NiagaraPosition=&AddVectorToPosition(
     Position=$Particles.Position
     Vector=&TransformVector(
      Vector=&MaskVector(Mask Source=0, Mask Parameter=?, Mask=0, Masked Vector=<0 0 23>, Passthrough Vector=<0 0 23>)
      Destination Space=::Simulation
     )
    )
    Particles.DustWidth:NiagaraFloat=&FloatFromCurve(
     FloatCurve=DataInterface() {propertyValues {Curve {
        defaultValue=1
        keys [
         {interpMode=RCIM_Cubic, tangentMode=RCTM_User, time=0.05, value=0.33}
         {interpMode=RCIM_Cubic, tangentMode=RCTM_User, time=0.26, value=1}
         {tangentMode=RCTM_User, time=0.7, value=0.66}
        ]
       }}}
     Scale Curve=83
    )
    Particles.GrungePosition:NiagaraPosition=&AddVectorToPosition(
     Position=$Particles.Position
     Vector=&TransformVector(
      Vector=&MaskVector(Mask Source=0, Mask Parameter=?, Mask=0, Masked Vector=<0 5 6>, Passthrough Vector=<0 -5 6>)
      Destination Space=::Simulation
     )
    )
    Particles.GrungeWidth:NiagaraFloat=&MaskFloat(Mask Source=0, Mask Parameter=?, Mask=0, Masked Float=-40, Passthrough Float=40)
    Particles.NoisePosition:NiagaraPosition=&AddVectorToPosition(
     Position=$Particles.Position
     Vector=&TransformVector(
      Vector=&MaskVector(Mask Source=0, Mask Parameter=?, Mask=0, Masked Vector=<0 -2 6>, Passthrough Vector=<0 2 6>)
      Destination Space=::Simulation
     )
    )
    Particles.NoiseWidth:NiagaraFloat=&MaskFloat(Mask Source=0, Mask Parameter=?, Mask=0, Masked Float=100, Passthrough Float=-100)
   ) {isSetParameters=true}
   &VC_UpdateRibbon_v1(Enable Skeletal Position=true)
  ]
  rendererClasses [&NiagaraRibbonRendererProperties]
  renderers [
   {rendererClass=&NiagaraRibbonRendererProperties}
   {rendererClass=&NiagaraRibbonRendererProperties, rendererIndex=1}
   {rendererClass=&NiagaraRibbonRendererProperties, rendererIndex=2}
   {rendererClass=&NiagaraRibbonRendererProperties, rendererIndex=3}
   {rendererClass=&NiagaraRibbonRendererProperties, rendererIndex=4}
  ]
 }
 NE_VC_WheelWrap {
  emitterUpdateScript [
   &EmitterState(
    Inactive Response=::NewEnumerator1
    Life Cycle Mode=::NewEnumerator1
    Loop Behavior=::NewEnumerator1
    Loop Duration Mode=::NewEnumerator1
    UseLoopDelay=false
   )
   &SpawnBurst_Instantaneous(Spawn Count=2, Use Spawn Probability=false, UseLoopCountLimit=false)
  ]
  particleSpawnScript [
   &VC_PaintedVariantsOverride(
    BurntSienna=<rgba 0.1042 0.0322 0.0128 1>
    DefaultColor=<rgba 1 0.4948 0.0017 1>
    ForestGreen=<rgba 0 0.25 0.0126 1>
    Lime=<rgba 0.2599 0.4583 0.0215 1>
    Orange=<rgba 0.3281 0.1095 0.0085 1>
    Pink=<rgba 0.2917 0.0349 0.1983 1>
    Purple=<rgba 0.1775 0.0382 0.4219 1>
    R1 Black:LinearColor=<rgba 0 0 0 2>
    R1 BurntSienna:LinearColor=<rgba 0.135 0.063 0.0315 2>
    R1 Cobalt:LinearColor=<rgba 0.0125 0.025 0.25 2.2>
    R1 Crimson:LinearColor=<rgba 0.215 0 0 2>
    R1 DefaultColor=<rgba 0 0 0 2>
    R1 ForestGreen:LinearColor=<rgba 0 0.18 0 2>
    R1 Grey:LinearColor=<rgba 0.1562 0.1562 0.1562 3>
    R1 Lime:LinearColor=<rgba 0.105 0.245 0 2>
    R1 Orange:LinearColor=<rgba 0.28 0.14 0.004 1>
    R1 Pink:LinearColor=<rgba 0.26 0.0685 0.1891 1>
    R1 Purple:LinearColor=<rgba 0.1375 0 0.165 2>
    R1 Saffron:LinearColor=<rgba 0.2 0.16 0 2.5>
    R1 SkyBlue:LinearColor=<rgba 0 0.1625 0.26 2>
    R1 TitaniumWhite:LinearColor=<rgba 0.4 0.4 0.45 1>
    Saffron=<rgba 0.4792 0.3674 0 1>
    SkyBlue=<rgba 0 0.2634 0.3958 1>
    TitaniumWhite=<rgba 0.224 0.224 0.224 1>
    Use R1 Painted Overrides=true
   )
   &InitializeParticle(
    Color=$Particles.Color
    Lifetime=$Emitter.CurrentLoopDuration
    Mesh Scale Mode=::NewEnumerator4
    Sprite Rotation Mode=::NewEnumerator2
    UsePositionOffset=false
    Write Material Random=true
   )
   &PartitionParticles()
   &SkeletalMeshLocation(
    Bone / Socket Index=&Add_Integer(A=$Particles.PartitionParticles.Partition, B=4)
    Bone / Socket Sampling Mode=::NewEnumerator5
    Interpolate Mesh Sampling=false
    MaskSpawnGroup=false
    Mesh Sampling Type=::NewEnumerator6
    Skeletal Mesh=$System.SkeletalMesh
   )
   /SetVariables(Particles.VisibilityTag=&ReturnExecIndex()) {isSetParameters=true}
  ]
  particleUpdateScript [
   &FN_DisableTireFX(
    Mask Index Attribute:NiagaraInt32=$Particles.PartitionParticles.Partition
    Mask Index FL=2
    Mask Index FR=3
    Mask Index RL=0
    Mask Index RR=1
   )
   &VC_WheelWrapScale(StandardWidth=30, Wheel Index Attribute:NiagaraInt32=&Add_Integer(A=$Particles.PartitionParticles.Partition, B=2))
   &DynamicMaterialParameters(
    Index 0 Param 1=3
    Index 0 Param 2=1
    Index 0 Param 3=2.5
    Index 0 Param 4=20
    Module.Index0Param0Write=true
    Module.Index0Param1Write=false
    Module.Index0Param2Write=false
    Module.Index0Param3Write=true
   ) {enabled=false}
   &ParticleState(Kill Particles When Lifetime Has Elapsed=false, Loop Particles Lifetime=true)
   &UpdateMeshOrientation(Rotation Coordinate Space=::NewEnumerator2, Rotation Vector=<0 1 0>) {enabled=false}
   &SkeletalMeshLocation(
    Bone / Socket Index=&Add_Integer(A=$Particles.PartitionParticles.Partition, B=4)
    Bone / Socket Sampling Mode=::NewEnumerator5
    MaskSpawnGroup=false
    Mesh Sampling Type=::NewEnumerator6
    Skeletal Mesh=$System.SkeletalMesh
    Spawn Group Mask=2
   ) {module=SkeletalMeshLocation001}
   &SolveForcesAndVelocity(Clamp Velocity=false, Limit Acceleration=false)
  ]
  rendererClasses [&NiagaraMeshRendererProperties]
  renderers [
   {rendererClass=&NiagaraMeshRendererProperties, rendererIndex=0}
   {rendererClass=&NiagaraMeshRendererProperties, rendererIndex=1}
  ]
 }
}
summary {
 systemName=NS_DriftTrail_RoadSparks
 emitters [
  {emitterName=NE_VC_WheelWrap, rendererClasses [&NiagaraMeshRendererProperties]}
  {emitterName=NE_VC_Trails, rendererClasses [&NiagaraRibbonRendererProperties]}
  {
   bEnabled=false
   emitterName=NE_VC_DriftTrail_Sprite_Parent001
   rendererClasses [&NiagaraSpriteRendererProperties, &NiagaraMeshRendererProperties]
  }
  {emitterName=NE_VC_DriftTrail_Cone, rendererClasses [&NiagaraSpriteRendererProperties]}
  {emitterName=NE_VC_DriftTrail_Sprites, rendererClasses [&NiagaraSpriteRendererProperties]}
  {emitterName=NE_VC_DriftTrail_ConeClouds001, rendererClasses [&NiagaraSpriteRendererProperties]}
 ]
 userVariables [
  User.bIsFrontendPreview {type=NiagaraBool}
  User.bIsPainted {type=NiagaraBool}
  User.blsIsCoroColor {type=NiagaraBool}
  User.bPreviewVehicleSpeed {type=NiagaraBool}
  User.Brightness2 {type=NiagaraFloat, defaultValue=1, description=""}
  User.ColorVariantID {type=NiagaraInt32}
  User.ConeColor {type=LinearColor, description="", defaultValue {struct=LinearColor, value=<rgba 1 1 1 1>}}
  User.DTCarOffset {type=NiagaraPosition, description="", defaultValue {}}
  User.DTCarOffset002 {type=NiagaraPosition, description="", defaultValue {}}
  User.GameModeID {type=NiagaraInt32}
  User.MultiVariantColorCone {type=NiagaraFloat, defaultValue=2.5, description=""}
  User.offset {type=NiagaraPosition, description="", defaultValue {}}
  User.offset {type=Vector3f, description="", defaultValue {struct=Vector3f, value=<0 0 0>}}
  User.offset001 {type=Vector3f, description="", defaultValue {struct=Vector3f, value=<0 0 0>}}
  User.offset002 {type=Vector3f, description="", defaultValue {struct=Vector3f, value=<0 0 0>}}
  User.offset003 {type=Vector3f, description="", defaultValue {struct=Vector3f, value=<0 0 0>}}
  User.PaintedColor {type=LinearColor, description="", defaultValue {struct=LinearColor, value=<rgba 1 1 1 1>}}
  User.PaintedColorTriangles {type=LinearColor, description="", defaultValue {struct=LinearColor, value=<rgba 1 1 1 1>}}
  User.PaintedScalar {type=NiagaraFloat, defaultValue=1, description=""}
  User.PowerTriColorPaintedCone {type=NiagaraFloat, defaultValue=1.5, description=""}
  User.VehicleFlags {type=NiagaraInt32, defaultValue=806289408}
  User.Wheel_Radius_Array {
   type=NiagaraDataInterfaceArrayFloat
   defaultValue {
    struct=NiagaraExt_VariableValue_DataInterface
    value {
     dataInterface=&self:NiagaraDataInterfaceArrayFloat_0
     dataInterfaceClass=&NiagaraDataInterfaceArrayFloat
    }
   }
  }
  User.Wheel_Width_Array {
   type=NiagaraDataInterfaceArrayFloat
   defaultValue {
    struct=NiagaraExt_VariableValue_DataInterface
    value {
     dataInterface=&self:NiagaraDataInterfaceArrayFloat_1
     dataInterfaceClass=&NiagaraDataInterfaceArrayFloat
    }
   }
  }
  User.WheelSize {type=NiagaraFloat, defaultValue=0.1, description=""}
 ]
}
